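# Material function for the BugHut environment effect that produces a slowly drifting wobble offset, likely for world position offset.
# It adds the WPO input to local position and scales the XY down heavily to form large-scale UVs, which pan over time.
# These UVs sample a spiral-cloud wave texture, whose RGB is recentered around zero and amplified as the output.
# The incoming WPO only shifts the sampling point and is not added to the result.
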MaterialFunction /Game/Effects/Fort_Effects/Effects/Environment/Bugs/BugHut/MF_BugHut_WobblePattern
graph {nodes [
  n1 = ComponentMask(None=Add(A=FunctionInput() {inputName=WPO}, B=LocalPosition().XYZ)) {g=true, r=true}
  n2 = Time()
  n3 = Constant2Vector() {g=-0.01, r=-0.008}
  n4 = TextureSample(UVs=Add(A=Divide(A=@n1) {constB=120000}, B=Multiply(A=@n2, B=@n3))) {samplerType=SAMPLERTYPE_LinearColor, texture=/Hermes/Techart/SpiralClouds/T_SpiralCloud_Wave}
  FunctionOutput(None=Multiply(A=Add(A=@n4.RGB) {constB=-0.5}) {constB=2500})
 ]}
