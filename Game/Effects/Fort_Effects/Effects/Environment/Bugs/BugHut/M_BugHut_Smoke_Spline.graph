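# Translucent, unlit, two-sided material for the BugHut environment effect's smoke plume.
# World position offset bends and widens the mesh along V, with an optional camera offset.
# Opacity blends three panning smoke textures, two swirl-distorted, with vertex-color masks, object-position variation, view, depth and distance fades, plus a distant-proxy switch.
# Color combines per-component random tints, time-of-day sunlight and fake wrap lighting from a normal map and sky light direction.
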Material /Game/Effects/Fort_Effects/Effects/Environment/Bugs/BugHut/M_BugHut_Smoke_Spline
graph {
 properties {
  bAutomaticallySetUsageInEditor=false
  blendMode=BLEND_Translucent
  shadingModel=MSM_Unlit
  translucencyLightingMode=TLM_VolumetricPerVertexDirectional
  twoSided=true
  usedShadingModels=Unlit
 }
 nodes [
  Constant3Vector() {constant=<rgba 0 0 1 0>}
  Constant()
  n1 = TextureSample(UVs=NamedRerouteUsage()) {samplerType=SAMPLERTYPE_Normal, texture=/Game/Effects/Fort_Effects/Textures/Smoke/T_Smoke_Trail_Soft_N}
  n2 = MaterialFunctionCall(Float2=TextureCoordinate()) {materialFunction=&BreakOutFloat2Components}
  n3 = ConstantBiasScale(None=@n2.R) {bias=-0.5, scale=3}
  n4 = ScalarParameter() {parameterName=Param_10}
  n5 = SphereMask(
   A=@n2.R
   B=ScalarParameter() {defaultValue=0.5, parameterName=Param_1}
   Radius=ScalarParameter() {defaultValue=0.3, parameterName=WrapLighting_Radius}
  ) {hardnessPercent=0}
  n6 = LinearInterpolate(
   A=Multiply(
    A=MaterialFunctionCall(X=@n1.G, Y=@n1.G, Z=@n1.G) {materialFunction=&MakeFloat3}.Result
    B=Constant3Vector() {constant=<rgba -1 -1 1 0>}
   )
   B=Normalize(VectorInput=MaterialFunctionCall(X=@n3, Y=@n4, Z=@n5) {materialFunction=&MakeFloat3}.Result)
   Alpha=ScalarParameter() {defaultValue=0.7, parameterName=Param_2}
  )
  n7 = DotProduct(
   A=Multiply(A=SkyAtmosphereLightDirection(), B=ScalarParameter() {defaultValue=1, parameterName=Param_7}) {constB=-1}
   B=Normalize(VectorInput=Transform(None=@n6))
  )
  n8 = Multiply(
   A=ScalarParameter() {defaultValue=1}
   B=LinearInterpolate(A=ScalarParameter() {defaultValue=0.1, parameterName=Param_11}, Alpha=ConstantBiasScale(None=@n7))
  ) {constA=1}
  MaterialFunctionCall(Emissive=Multiply(A=@n8, B=NamedRerouteUsage())) {materialFunction=&MF_FX_EmissiveAdjustments}
  Multiply(
   A=MaterialFunctionCall() {materialFunction=/Game/Packages/Fortress_SharedMaterials/Base_Material_Functions/SkylightColorAndIntensity.SkyLightColorAndIntensity}.Color
   B=@n8
  )
  n9 = VertexColor()
  n10 = Frac(None=Multiply(A=@n9.A) {constB=37.714})
  NamedRerouteDeclaration(None=Step(X=@n10) {constY=0.5}) {name=PerComponentRandom, nodeColor=<rgba 0.1403 1 0.8629 1>}
  n11 = NamedRerouteUsage()
  n12 = Multiply(A=@n11, B=VectorParameter() {parameterName=Smoke Base Color, defaultValue=<rgba 1.0186 0 1.3 1>}.RGB)
  n13 = Multiply(A=@n11, B=Constant3Vector() {constant=<rgba 0.8173 0 1.5 1>})
  n14 = NamedRerouteUsage()
  NamedRerouteDeclaration(None=Multiply(A=LinearInterpolate(A=@n12, B=@n13, Alpha=@n14), B=ScalarParameter() {defaultValue=0.5, parameterName=Albedo Scale})) {name=Base Color, nodeColor=<rgba 1 0.1406 0 1>}
  NamedRerouteDeclaration(None=MaterialFunctionCall(
    Color_Morning=VectorParameter() {parameterName=Sunlight Morning, defaultValue=<rgba 0.6604 0.6604 0.6604 1>}.RGB
    Color_Day=VectorParameter() {parameterName=Sunlight Day, defaultValue=<rgba 1 1 1 1>}.RGB
    Color_Evening=VectorParameter() {parameterName=Sunlight Evening, defaultValue=<rgba 1 1 1 1>}.RGB
    Color_Night=VectorParameter() {parameterName=Sunlight Night, defaultValue=<rgba 1 1 1 0>}.RGB
   ) {materialFunction=&Color_Time_of_Day}.Result) {name=directionality, nodeColor=<rgba 1 0.6328 0 1>}
  n15 = TextureCoordinate()
  n16 = ComponentMask(None=VectorParameter() {parameterName=edge distort tile, defaultValue=<rgba 1 1 0 0>}.RGB) {g=true, r=true}
  n17 = Time()
  n18 = ScalarParameter() {defaultValue=0.1, parameterName=Param_16}
  n19 = TextureSample(UVs=Panner(Coordinate=Multiply(A=@n15, B=@n16), Time=Multiply(A=@n17, B=@n18)) {speedY=1}) {samplerType=SAMPLERTYPE_Normal, texture=/Game/Effects/Fort_Effects/Emotes/BirdNestBlue/T_SwirlNoise}
  n20 = ComponentMask(None=TextureCoordinate()) {g=true}
  n21 = Multiply(
   A=ComponentMask(None=MaterialFunctionCall(Normal=@n19.RGB, Flatness=@n20) {materialFunction=&FlattenNormal}.Result) {g=true, r=true}
   B=ScalarParameter() {defaultValue=0.02, parameterName=Param_15}
  )
  n22 = TextureCoordinate()
  n23 = ComponentMask(None=VectorParameter() {parameterName=Tiling 1, defaultValue=<rgba 1 2 0 0>}.RGB) {g=true, r=true}
  n24 = Panner(
   Coordinate=Add(A=@n21, B=Multiply(A=@n22, B=@n23))
   Time=Multiply(A=Time(), B=ScalarParameter() {defaultValue=0.125, parameterName=Time Scale 1})
  ) {speedY=1}
  n25 = Multiply(
   A=TextureSample(UVs=MaterialFunctionCall(XY=@n24) {materialFunction=&Swizzle}.YX) {samplerType=SAMPLERTYPE_Grayscale, texture=/Game/Effects/Fort_Effects/Textures/Smoke/T_Smoke_Trail_Soft_NoClamp}.G
   B=ScalarParameter() {defaultValue=0.25, parameterName=Texture 1 Opacity}
  )
  n26 = TextureCoordinate()
  n27 = ComponentMask(None=VectorParameter() {parameterName=Tiling 2, defaultValue=<rgba 1 2 0 0>}.RGB) {g=true, r=true}
  n28 = Panner(
   Coordinate=Add(A=@n21, B=Multiply(A=@n26, B=@n27))
   Time=Multiply(A=Time(), B=ScalarParameter() {defaultValue=0.1, parameterName=Time Scale 2})
  ) {speedY=1}
  n29 = Multiply(
   A=TextureSample(UVs=MaterialFunctionCall(XY=@n28) {materialFunction=&Swizzle}.YX) {samplerType=SAMPLERTYPE_Grayscale, texture=&T_Smoke_Trail_Soft_NoClamp}.G
   B=ScalarParameter() {defaultValue=1, parameterName=Texture 2 Opacity}
  )
  n30 = TextureCoordinate()
  n31 = ComponentMask(None=LinearInterpolate(
    A=VectorParameter() {parameterName=Tiling 3, defaultValue=<rgba 1 1.5 0 0>}.RGB
    B=Constant3Vector() {constant=<rgba 1 1 0 0>}
    Alpha=NamedRerouteUsage()
   )) {g=true, r=true}
  n32 = Panner(Coordinate=Add(B=Multiply(A=@n30, B=@n31)), Time=Multiply(A=Time(), B=ScalarParameter() {defaultValue=0.1, parameterName=Time Scale 3})) {speedY=1}
  n33 = TextureSample(UVs=NamedRerouteDeclaration(None=MaterialFunctionCall(XY=@n32) {materialFunction=&Swizzle}.YX) {name=UV3, nodeColor=<rgba 0 0.4844 1 1>}) {samplerType=SAMPLERTYPE_Grayscale, texture=&T_Smoke_Trail_Soft_NoClamp}
  n34 = ScalarParameter() {defaultValue=1, parameterName=Texture 3 Density}
  n35 = Power(
   Base=LinearInterpolate(A=LinearInterpolate(A=@n25, B=@n29), B=Multiply(A=@n33.G, B=@n34))
   Exp=ScalarParameter() {defaultValue=3, parameterName=Blend Power}
  )
  n36 = MaterialFunctionCall("FadeSensitivity (optional)"=ScalarParameter() {defaultValue=1.5, parameterName=Angle Fade}) {materialFunction=&FadeBasedOnViewAngle}
  n37 = OneMinus(None=ComponentMask(None=TextureCoordinate()) {g=true})
  n38 = ScalarParameter() {parameterName=Tip Fade}
  n39 = Multiply(
   A=Multiply(A=Multiply(A=@n35, B=@n36.Result), B=Power(Base=@n37, Exp=@n38))
   B=MaterialFunctionCall(Distance=ScalarParameter() {defaultValue=150, parameterName=Near Camera Fade}) {materialFunction=&Near_Camera_Fade}.Result
  )
  n40 = ScalarParameter() {defaultValue=50, parameterName=Depth Fade}
  n41 = ObjectPositionWS()
  n42 = Floor(None=Fmod(A=Multiply(A=@n41) {constB=17.123}, B=Constant() {r=3}))
  n43 = LinearInterpolate(A=LinearInterpolate(A=@n9.R, B=@n9.R, Alpha=Saturate(None=@n42)), B=@n9.B, Alpha=Step(X=@n42) {constY=1.5})
  n44 = ScalarParameter() {defaultValue=1, parameterName=OverallOpacity}
  n45 = LinearInterpolate(A=Subtract(A=@n44) {constB=0.2}, B=@n44, Alpha=@n10) {constA=0.4, constB=0.6}
  n46 = WorldPosition()
  n47 = CameraPositionWS()
  n48 = Multiply(
   A=Multiply(A=DepthFade(Opacity=@n39, FadeDistance=@n40), B=Multiply(A=@n43, B=@n45))
   B=LinearInterpolate(
    A=ScalarParameter() {defaultValue=8, parameterName=Opacity Far}
    B=ScalarParameter() {defaultValue=10, parameterName=Opacity Close}
    Alpha=SphereMask(A=@n46.XYZ, B=@n47) {attenuationRadius=15000, hardnessPercent=10}
   )
  )
  n49 = Subtract(A=PixelDepth().R, B=ScalarParameter() {parameterName=Fade Distance})
  n50 = ScalarParameter() {parameterName=Fade Soften}
  n51 = Multiply(A=@n48, B=Saturate(None=Divide(A=@n49, B=@n50)))
  n52 = StaticBoolParameter() {parameterName="Distant Proxy, Epic"}
  n53 = ComponentMask(None=TextureCoordinate()) {g=true}
  n54 = ScalarParameter() {defaultValue=1, parameterName=End Fade}
  n55 = LinearInterpolate(A=@n54, B=Add(A=@n54) {constB=0.5}, Alpha=NamedRerouteUsage()) {constB=2}
  NamedRerouteDeclaration(None=Multiply(A=StaticSwitch(True=@n51, False=@n48, Value=@n52), B=Power(Base=@n53, Exp=@n55))) {name=Opacity, nodeColor=<rgba 0 0.1797 1 1>}
  n56 = ComponentMask(None=TextureCoordinate()) {g=true}
  n57 = ScalarParameter() {defaultValue=1, parameterName=Bend Power}
  n58 = LinearInterpolate(
   A=VectorParameter() {parameterName=root}.RGB
   B=VectorParameter() {parameterName=Bend, defaultValue=<rgba -500 0 0 0>}.RGB
   Alpha=OneMinus(None=Power(Base=@n56, Exp=@n57))
  )
  n59 = ComponentMask(None=TextureCoordinate()) {r=true}
  n60 = ComponentMask(None=TextureCoordinate()) {g=true}
  n61 = Multiply(
   A=Multiply(A=Subtract(A=@n59) {constB=0.5}, B=MaterialFunctionCall() {materialFunction=&TangentBasis}.X U)
   B=LinearInterpolate(
    A=ScalarParameter() {defaultValue=150, parameterName=Base}
    B=ScalarParameter() {defaultValue=3500, parameterName=End}
    Alpha=OneMinus(None=@n60)
   )
  )
  n62 = ScalarParameter() {parameterName=CameraOffset}
  n63 = WorldPosition()
  NamedRerouteDeclaration(None=Add(
    A=Add(A=@n58, B=@n61)
    B=MaterialFunctionCall(Offset Amount=@n62, World Position=@n63.XYZ) {materialFunction=/Engine/Functions/Engine_MaterialFunctions02/WorldPositionOffset/CameraOffset}.Camera Offset
   )) {name=WPO, nodeColor=<rgba 0.5 1 0 1>}
  NamedRerouteUsage()
  NamedRerouteUsage()
  Normalize(VectorInput=VectorParameter() {group=Shading, parameterName=SunDirectionVector, defaultValue=<rgba 0 0 1 1>}.RGB)
 ]
}
