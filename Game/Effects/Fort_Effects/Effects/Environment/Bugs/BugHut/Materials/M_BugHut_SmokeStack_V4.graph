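# Unlit, masked, two-sided material for the BugHut smoke stack, animating sprite cards as a looping, cyan-tinted plume.
# Time and position within instance bounds drive a looping age that moves cards, while size varies along the plume and with depth.
# Spherical-harmonic coefficients sampled from sliced volume textures are decoded for sun shading and transmittance, with time-of-day colors and hue shift.
# A noise-distorted smoke mask with Bayer dithering shapes opacity.
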Material /Game/Effects/Fort_Effects/Effects/Environment/Bugs/BugHut/Materials/M_BugHut_SmokeStack_V4
graph {
 properties {bAutomaticallySetUsageInEditor=false, blendMode=BLEND_Masked, shadingModel=MSM_Unlit, twoSided=true, usedShadingModels=Unlit}
 nodes [
  Add()
  n1 = ScalarParameter() {defaultValue=0.17, parameterName=Shadow Scale Morning}
  n2 = ScalarParameter() {defaultValue=0.5, parameterName=Param_23}
  n3 = Multiply(
   A=NamedRerouteUsage()
   B=MaterialFunctionCall(
    Value_Morning=MaterialFunctionCall!("Xb1,PS4"=@n1, "PC-High&&Epic,XSX,PS5"=@n1, "PC-Perf,Android"=@n2, Switch=@n1, "PC-Low(NotPerf)&&Medium"=@n1) {materialFunction=/Game/Effects/Niagara/MasterMaterials/Functions/Common/PlatformSwitch_3D}.Result
    Value_Day=ScalarParameter() {defaultValue=0.35, parameterName=Shadow Scale}
    Value_Evening=ScalarParameter() {defaultValue=0.05, parameterName=Shadow Scale Evening}
    Value_Night=ScalarParameter() {defaultValue=0.35, parameterName=Shadow Ccale Night}
   ) {materialFunction=&Scalar_Time_of_Day}.Result
  ) {constB=0.35}
  n4 = Multiply(A=Power(Base=NamedRerouteUsage()), B=NamedRerouteUsage())
  n5 = LinearInterpolate(A=@n3, B=NamedRerouteUsage(), Alpha=Power(Base=Saturate(None=@n4), Exp=ScalarParameter() {defaultValue=1.7, parameterName=Param_17}))
  n6 = MaterialFunctionCall(Hue Shift Percentage=ScalarParameter() {defaultValue=0.85, parameterName=HueShift}, Texture=@n5) {materialFunction=&HueShift}
  n7 = NamedRerouteUsage()
  n8 = OneMinus(None=Power(Base=ComponentMask(None=@n7) {r=true}, Exp=ScalarParameter() {defaultValue=2, parameterName=Hue Shift Gradient}))
  n9 = NamedRerouteUsage()
  n10 = Power(Base=NamedRerouteUsage(), Exp=ScalarParameter() {defaultValue=1, parameterName=Additional SH Directionality Power})
  Add(
   A=Multiply(A=NamedRerouteUsage(), B=ScalarParameter() {defaultValue=1, parameterName=Param_12})
   B=Add(A=LinearInterpolate(A=@n6.Result, B=@n5, Alpha=@n8), B=Multiply(A=@n9, B=@n10))
  )
  n11 = Multiply(A=NamedRerouteUsage(), B=NamedRerouteUsage())
  n12 = MaterialFunctionCall() {materialFunction=&ObjectScale}
  n13 = Add(
   A=Subtract(A=ActorPositionWS() {originType=CameraRelative}, B=Multiply(A=@n11, B=@n12.Scale XYZ))
   B=Multiply(A=NamedRerouteUsage(), B=@n12.Scale XYZ)
  )
  Add(A=@n13, B=Multiply(A=Transform() {transformSourceType=TRANSFORMSOURCE_World, transformType=TRANSFORM_Local}, B=NamedRerouteUsage()))
  Add()
  Add()
  n14 = NamedRerouteUsage()
  n15 = Constant()
  Add(B=MaterialFunctionCall("Xb1,PS4"=@n14, "PC-High&&Epic,XSX,PS5"=@n14, "PC-Perf,Android"=@n15, Switch=@n15, "PC-Low(NotPerf)&&Medium"=@n15) {materialFunction=&PlatformSwitch_3D}.Result)
  n16 = Divide(B=VectorParameter() {parameterName=Bounding Box Size, defaultValue=<rgba 700 700 500 1>}.RGB)
  n17 = Normalize(VectorInput=VectorParameter() {parameterName=Y, defaultValue=<rgba 0 1 0 1>}.RGB)
  n18 = CrossProduct(A=@n17, B=VectorParameter() {parameterName=Z, defaultValue=<rgba 0 0 1 1>}.RGB)
  n19 = CrossProduct(A=@n18, B=@n17)
  n20 = Divide(A=LocalPosition().XYZ) {constB=200}
  n21 = TextureObject() {samplerType=SAMPLERTYPE_LinearColor, texture=/Game/Environments/Helios/TechArt/LowAltitudeClouds/T_LowAltitudeClouds_N}
  n22 = Constant() {r=8}
  n23 = ScalarParameter() {defaultValue=0.9375, parameterName=AtlasPadding, primitiveDataIndex=8}
  Add(
   A=NamedRerouteDeclaration(None=MaterialFunctionCall!(VectorToTransform=@n16, BasisX=@n18, BasisY=@n17, BasisZ=@n19) {materialFunction=/Engine/Functions/Engine_MaterialFunctions02/Math/InverseTransformMatrix}.Transformed Vector 3) {name=All World, nodeColor=<rgba 1 0 0.1406 1>}
   B=Add(A=MaterialFunctionCall!(UVs=@n20, Slices Texture=@n21, Atlas Size=@n22, Atlas Padding=@n23) {materialFunction=/Game/Environments/Helios/TechArt/LowAltitudeClouds/MF_3DVolumeFrom2DSlices}.RGB) {constB=-0.5}
  )
  Add()
  Add()
  CollectionParameter() {collection=&FortniteMaterialParameters, parameterName=RemappedTimeOfDay_0-1}
  n24 = NamedRerouteUsage()
  n25 = Divide(A=Transform(None=ConstantBiasScale(None=@n24) {bias=-0.5, scale=2}) {transformSourceType=TRANSFORMSOURCE_Local}, B=NamedRerouteUsage())
  n26 = ScalarParameter() {defaultValue=-0.05, parameterName=Box Bias Offset}
  n27 = Add(
   A=Multiply(A=Add(A=@n25, B=@n26) {constB=-0.02}) {constB=2}
   B=Multiply(A=NamedRerouteUsage(), B=ScalarParameter() {defaultValue=1, parameterName=Noise Scale})
  )
  ComponentMask(None=@n27) {g=true, r=true}
  ComponentMask() {b=true, g=true, r=true}
  n28 = CollectionParameter() {
   collection=&FortniteMaterialParameters
   parameterName=FortniteActiveTimeOfDay
  }
  ComponentMask(None=@n28) {a=true}
  ComponentMask() {r=true}
  ComponentMask() {g=true, r=true}
  ComponentMask() {a=true}
  n29 = Sine(None=Multiply(A=Time(), B=ScalarParameter() {defaultValue=0.025}))
  ConstantBiasScale(None=@n29) {bias=-0.5, scale=2}
  Constant() {r=1}
  Desaturation()
  n30 = SkyAtmosphereLightDirection()
  DotProduct(A=@n30)
  n31 = Divide(A=Transform(None=NamedRerouteUsage()) {transformSourceType=TRANSFORMSOURCE_Local}, B=NamedRerouteUsage())
  n32 = ScalarParameter() {defaultValue=-0.02, parameterName=Param_42}
  DotProduct(A=Multiply(A=Add(A=@n31, B=@n32) {constB=0.5}) {constB=2}, B=@n30)
  DotProduct(A=ConstantBiasScale(None=NamedRerouteUsage()) {bias=-0.5, scale=2}, B=SkyAtmosphereLightDirection())
  DotProduct(A=NamedRerouteUsage(), B=AtmosphericLightVector())
  LocalPosition()
  MaterialFunctionCall() {materialFunction=&ObjectScale}
  MaterialFunctionCall() {materialFunction=&ObjectScale}
  MaterialFunctionCall() {materialFunction=&DitherTemporalAA}
  MaterialFunctionCall() {materialFunction=&Scalar_Time_of_Day}
  n33 = ActorPositionWS() {originType=CameraRelative}
  n34 = NamedRerouteUsage()
  n35 = Add(A=Add(A=WorldPosition().XYZ, B=@n34), B=NamedRerouteUsage())
  n36 = Divide(
   A=Transform(None=Subtract(A=@n33, B=@n35)) {transformSourceType=TRANSFORMSOURCE_World, transformType=TRANSFORM_Instance}
   B=Multiply(A=MaterialFunctionCall() {materialFunction=&ObjectScale}.Scale XYZ) {constB=100}
  )
  Multiply(A=Power(Base=Normalize(VectorInput=@n36), Exp=ScalarParameter() {defaultValue=1, parameterName=Param_5}), B=NamedRerouteUsage())
  Multiply() {constB=-1}
  n37 = VertexInterpolator(VS=LocalPosition() {localOrigin=InstancePreSkinning}.XYZ)
  n38 = VectorParameter() {parameterName=Param_5, defaultValue=<rgba -500 0 0 0>}
  n39 = WorldPosition() {worldPositionShaderOffset=WPT_CameraRelative}
  n40 = NamedRerouteUsage()
  n41 = Divide(
   A=Subtract(A=Subtract(A=@n37.PS, B=@n38.RGB), B=Add(A=@n39.XYZ, B=@n40))
   B=MaterialFunctionCall() {materialFunction=&ObjectScale}.Scale XYZ
  )
  n42 = Transform(None=Subtract(A=WorldPosition() {worldPositionShaderOffset=WPT_CameraRelative}.XYZ, B=Add(A=@n13, B=@n41))) {transformSourceType=TRANSFORMSOURCE_World, transformType=TRANSFORM_Local}
  n43 = Saturate(None=DotProduct(A=Normalize(VectorInput=@n42), B=AtmosphericLightVector()))
  n44 = ScalarParameter() {defaultValue=1, parameterName=Param_25}
  Multiply(A=Multiply(A=Power(Base=@n43, Exp=@n44), B=ScalarParameter() {defaultValue=1, parameterName=Param_13}), B=NamedRerouteUsage())
  Multiply(B=NamedRerouteUsage())
  Multiply()
  Multiply() {constA=-1, constB=-1}
  Multiply(A=VectorParameter() {parameterName=Param_6, defaultValue=<rgba 1 1 0 0>}.RGB)
  Multiply() {constB=100}
  n45 = Custom!(None) {code="GetPrimitiveData(Parameters).InstanceLocalBoundsExtent.xyz", description=BoundsExtent}
  n46 = Subtract(
   A=TransformPosition(None=WorldPosition() {worldPositionShaderOffset=WPT_ExcludeAllShaderOffsets}.XYZ) {transformSourceType=TRANSFORMPOSSOURCE_World, transformType=TRANSFORMPOSSOURCE_Instance}
   B=Subtract(A=Custom() {code="GetPrimitiveData(Parameters).InstanceLocalBoundsCenter.xyz", description=BoundsCenter}, B=@n45)
  )
  n47 = Multiply(A=@n45) {constB=2}
  n48 = MaterialFunctionCall(Float3=NamedRerouteDeclaration(None=Divide(A=@n46, B=@n47)) {name=BBX, nodeColor=<rgba 1 0 0.6094 1>}) {materialFunction=&BreakOutFloat3Components}
  n49 = Multiply(A=Add(A=Time(), B=NamedRerouteUsage()), B=ScalarParameter() {defaultValue=0.005, parameterName=Time Scale})
  n50 = ScalarParameter() {defaultValue=0.1, parameterName=Loop Duration}
  n51 = Divide(A=Fmod(A=Add(A=@n48.R, B=@n49), B=@n50), B=@n50)
  NamedRerouteDeclaration(None=@n51) {name=NOrmalized Age, nodeColor=<rgba 0.4766 0 1 1>}
  NamedRerouteDeclaration(None=SphereMask(
    A=NamedRerouteUsage()
    B=ScalarParameter() {defaultValue=0.5, parameterName=Param_1}
    Radius=ScalarParameter() {defaultValue=0.5, parameterName=Param_2}
    Hardness=ScalarParameter() {defaultValue=0.25, parameterName=Param_3}
   )) {name=Opacity, nodeColor=<rgba 0 1 0.1562 1>}
  n52 = Constant4Vector() {constant=<rgba 1.075 -0.33 -0.333 -0.344>}
  n53 = Subtract(A=Constant4Vector() {constant=<rgba 1.579 0.32 0.34 0.287>}.RGBA, B=@n52.RGBA)
  n54 = Add(A=Add(A=WorldPosition() {worldPositionShaderOffset=WPT_ExcludeAllShaderOffsets}.XYZ, B=NamedRerouteUsage()), B=NamedRerouteUsage())
  n55 = Custom() {code="GetPrimitiveData(Parameters).InstanceLocalBoundsExtent.xyz", description=BoundsExtent}
  n56 = Subtract(
   A=TransformPosition(None=@n54) {transformSourceType=TRANSFORMPOSSOURCE_World, transformType=TRANSFORMPOSSOURCE_Instance}
   B=Subtract(A=Custom() {code="GetPrimitiveData(Parameters).InstanceLocalBoundsCenter.xyz", description=BoundsCenter}, B=@n55)
  )
  n57 = Multiply(A=@n55) {constB=2}
  n58 = MaterialFunctionCall(Float3=OneMinus(None=Divide(A=@n56, B=@n57))) {materialFunction=&BreakOutFloat3Components}
  n59 = Subtract(A=MaterialFunctionCall(X=@n58.R, Y=@n58.R, Z=@n58.R) {materialFunction=&MakeFloat3}.Result) {constB=0.5}
  n60 = VectorParameter() {parameterName=Sphere Scale, defaultValue=<rgba 0.5 0.5 0.5 0>}
  n61 = TextureObject() {
   samplerType=SAMPLERTYPE_LinearColor
   texture=/Game/Effects/Fort_Effects/GameplayPulginsAssets/Doom/DoomCastle/Smoke/MaterialAndMesh/T_DoomSmokeSH
  }
  n62 = MaterialFunctionCall!(UVs=Add(A=Multiply(A=@n59, B=@n60.RGB)) {constB=0.5}, Slices Texture=@n61, Atlas Size=Constant() {r=4}, Atlas Padding=Constant() {r=0.95}) {materialFunction=/Game/Effects/Fort_Effects/GameplayPulginsAssets/Doom/DoomCastle/Smoke/MaterialAndMesh/MF_DoomCastle_3DVolumeFrom2DSlices}
  n63 = AppendVector(A=@n62.RGB, B=@n62.RGB)
  n64 = NamedRerouteDeclaration(None=Add(A=Multiply(A=@n53, B=@n63), B=@n52.RGBA)) {name=SH_Coefficients, nodeColor=<rgba 0.6641 1 0 1>}
  n65 = VertexInterpolator(VS=MaterialFunctionCall!(Direction=SkyAtmosphereLightDirection(), SH=@n64) {materialFunction=/Game/Effects/Fort_Effects/GameplayPulginsAssets/Doom/DoomCastle/Smoke/MaterialAndMesh/MF_SHDecode}.Result)
  n66 = Power(
   Base=MaterialFunctionCall(
    Input=OneMinus(None=@n65.PS)
    Input Low=ScalarParameter() {defaultValue=0.25, parameterName=Remap SH Low}
    Input High=ScalarParameter() {defaultValue=1, parameterName=Remap SH High}
   ) {materialFunction=&RemapValueRange}.Result
   Exp=MaterialFunctionCall(
    Value_Morning=ScalarParameter() {defaultValue=2.6544, parameterName=SH2Contrast_HighValue_ Morning}
    Value_Day=ScalarParameter() {defaultValue=1.75, parameterName=SH2Contrast_HighValue Day}
    Value_Evening=ScalarParameter() {defaultValue=1.5, parameterName=SH2Contrast_HighValue Evening}
    Value_Night=ScalarParameter() {defaultValue=1, parameterName=SH2Contrast_HighValue Night}
   ) {materialFunction=&Scalar_Time_of_Day}.Result
  )
  n67 = ScalarParameter() {defaultValue=1, parameterName=SH Intensity}
  NamedRerouteDeclaration(None=Saturate(None=Multiply(A=@n66, B=@n67))) {name=Decoded SH, nodeColor=<rgba 1 0.0234 0 1>}
  NamedRerouteDeclaration(None=MaterialFunctionCall(
    Color_Morning=VectorParameter() {parameterName=Param_1, defaultValue=<rgba 0 0.1653 0.2066 1>}.RGB
    Color_Day=VectorParameter() {parameterName=Albedo, defaultValue=<rgba 0.0529 0.4344 0.5399 1>}.RGB
    Color_Evening=VectorParameter() {parameterName=Param_4, defaultValue=<rgba 0 0.2878 0.3837 1>}.RGB
    Color_Night=VectorParameter() {parameterName=Albedo Night, defaultValue=<rgba 0 0.35 0.6 1>}.RGB
   ) {materialFunction=&Color_Time_of_Day}.Result) {name=Albedo, nodeColor=<rgba 0.9453 0 1 1>}
  n68 = Add(A=WorldPosition() {worldPositionShaderOffset=WPT_CameraRelative}.XYZ, B=NamedRerouteUsage())
  n69 = NamedRerouteUsage()
  n70 = Subtract(A=Subtract(A=Add(A=@n68, B=@n69), B=NamedRerouteUsage()), B=ActorPositionWS() {originType=CameraRelative})
  n71 = Divide(
   A=@n70
   B=MaterialFunctionCall() {materialFunction=&ObjectScale}.Scale XYZ
  )
  NamedRerouteDeclaration(None=DotProduct(A=@n71, B=AtmosphericLightVector())) {name=Mask, nodeColor=<rgba 1 0.8906 0 1>}
  n72 = DotProduct(A=Multiply(A=AtmosphericLightVector()) {constB=-1}, B=NamedRerouteUsage())
  n73 = ScalarParameter() {defaultValue=0.5, parameterName=Detail Intensity}
  n74 = Power(
   Base=Add(A=Multiply(A=@n72, B=@n73), B=ScalarParameter() {defaultValue=0.5, parameterName=Param_15}) {constB=0.5}
   Exp=ScalarParameter() {defaultValue=1.2, parameterName=Param_16}
  )
  NamedRerouteDeclaration(None=LinearInterpolate(A=@n3, B=NamedRerouteUsage(), Alpha=Saturate(None=@n74))) {name=Detail Pass, nodeColor=<rgba 1 0.2578 0 1>}
  NamedRerouteDeclaration(None=@n43) {name=Normals, nodeColor=<rgba 0 1 0.5781 1>}
  n75 = WorldPosition()
  n76 = NamedRerouteUsage()
  n77 = Subtract(A=Add(A=Add(A=@n75.XYZ, B=@n76), B=NamedRerouteUsage()), B=CameraPositionWS())
  n78 = MaterialFunctionCall(Direction=Transform(None=Normalize(VectorInput=@n77)) {transformSourceType=TRANSFORMSOURCE_World, transformType=TRANSFORM_Local}, SH=@n64) {materialFunction=&MF_SHDecode}
  NamedRerouteDeclaration(None=VertexInterpolator(VS=OneMinus(None=@n78.Result)).PS) {name=Transmittance, nodeColor=<rgba 1 0 0.7031 1>}
  n79 = ScalarParameter() {defaultValue=-1, parameterName=Param_20}
  n80 = ScalarParameter() {defaultValue=4017869.25, parameterName=Param_18}
  n81 = ScalarParameter() {defaultValue=1, parameterName=Param_21}
  n82 = ScalarParameter() {defaultValue=1, parameterName=Param_22}
  n83 = CollectionParameter() {
   collection=&FortniteMaterialParameters
   parameterName=FortniteActiveTimeOfDay
  }
  n84 = ComponentMask(None=@n83) {a=true}
  n85 = Multiply(
   A=Add(
    A=NamedRerouteUsage()
    B=MaterialFunctionCall!(1=@n79, 2=@n80, 3=@n81, 4=@n82, A=@n84) {materialFunction=/Engine/Functions/Engine_MaterialFunctions02/Lerp_Multiple_Float}.Lerp 4 Inputs
   ) {constB=-0.5}
   B=ScalarParameter() {defaultValue=1, parameterName=Param_19}
  ) {constB=2}
  n86 = AtmosphericLightVector()
  NamedRerouteDeclaration(None=Saturate(None=DotProduct(A=@n85, B=@n86))) {name=BBX Shading, nodeColor=<rgba 0 0.7656 1 1>}
  n87 = OneMinus(None=ComponentMask(None=NamedRerouteUsage()) {r=true})
  n88 = ScalarParameter() {defaultValue=0.5, parameterName=Param_7}
  n89 = Add(A=Subtract(A=TextureCoordinate(), B=Constant() {r=0.75}), B=VertexColor().G)
  n90 = Constant() {r=0.25}
  n91 = Time()
  n92 = Constant() {r=0.025}
  n93 = TextureSample(UVs=Panner(Coordinate=Multiply(A=@n89, B=@n90), Time=Multiply(A=@n91, B=@n92)) {speedY=1}) {texture=/Game/VisualThreatManager/StormVisuals/Textures/2ndInfectedMapLinear}
  n94 = Multiply(A=@n93.G, B=Constant() {r=0.15})
  n95 = TextureCoordinate()
  n96 = TextureSample(UVs=Subtract(A=Add(A=@n94, B=@n95), B=ScalarParameter() {defaultValue=0.1, parameterName=Param_45})) {texture=/Game/Effects/Fort_Effects/GameplayPulginsAssets/Doom/DoomCastle/Smoke/MaterialAndMesh/SmokeMask_02_T2}
  n97 = ScalarParameter() {defaultValue=-0.041, parameterName=Param_10}
  n98 = Constant() {r=1}
  n99 = Constant() {r=0.7}
  n100 = Multiply(
   A=Power(
    Base=Subtract(A=@n96.G, B=@n97)
    Exp=MaterialFunctionCall("Xb1,PS4"=@n98, "PC-High&&Epic,XSX,PS5"=@n98, "PC-Perf,Android"=@n99, Switch=@n99, "PC-Low(NotPerf)&&Medium"=@n98) {materialFunction=&PlatformSwitch_3D}.Result
   )
   B=NamedRerouteUsage()
  )
  NamedRerouteDeclaration(None=Multiply(
    A=Power(Base=@n87, Exp=@n88)
    B=MaterialFunctionCall!(In=@n100, Iterations, Scale) {materialFunction=/Game/Packages/Fortress_SharedMaterials/Base_Material_Functions/MF_DitherBayerNoTex}.Result
   )) {name=Final Mask, nodeColor=<rgba 0.2422 1 0 1>}
  n101 = Divide(A=PixelDepth().R, B=ScalarParameter() {defaultValue=350, parameterName=Shrink over Distance})
  n102 = ScalarParameter() {defaultValue=0.55, parameterName=Length Scale Power}
  n103 = Multiply(
   A=LinearInterpolate(A=ScalarParameter() {parameterName=Param_29}, Alpha=Saturate(None=@n101))
   B=LinearInterpolate(
    A=ScalarParameter() {parameterName=Scale Min}
    B=ScalarParameter() {defaultValue=3500, parameterName=Scale Max}
    Alpha=Power(Base=@n48.R, Exp=@n102)
   )
  )
  n104 = MaterialFunctionCall!(Center Location=WorldPosition().XYZ, 0-1 UVs=TextureCoordinate(), XY scale=@n103, OverridePivot, "Normals (Optional)") {materialFunction=&Sprite}
  NamedRerouteDeclaration(None=@n104.Normals) {name=Normal Sprite, nodeColor=<rgba 0.5 0 1 1>}
  NamedRerouteDeclaration(None=@n103) {name=Size, nodeColor=<rgba 0 1 0.1797 1>}
  n105 = Multiply(A=@n51, B=VectorParameter() {parameterName=Movement Dir, defaultValue=<rgba 850 750 500 0>}.RGB)
  NamedRerouteDeclaration(None=@n105) {name=Movements, nodeColor=<rgba 0.5937 1 0 1>}
  n106 = Add(A=VertexInterpolator(VS=PreSkinnedNormal()).PS, B=@n34)
  n107 = NamedRerouteUsage()
  n108 = LocalPosition() {localOrigin=InstancePreSkinning}
  n109 = Subtract(A=Add(B=VertexInterpolator(VS=@n108.XYZ).PS), B=@n35)
  n110 = NamedRerouteUsage()
  n111 = Transform(None=Subtract(A=Subtract(A=@n106, B=@n107), B=Divide(A=@n109, B=@n110) {constB=1000})) {transformSourceType=TRANSFORMSOURCE_Local}
  n112 = CameraPositionWS()
  NamedRerouteDeclaration(None=Normalize(VectorInput=Subtract(A=@n111, B=@n112))) {name=Card Vector, nodeColor=<rgba 1 0 0.7969 1>}
  n113 = Saturate(None=DotProduct(A=@n27, B=AtmosphericLightVector()))
  n114 = ScalarParameter() {defaultValue=2, parameterName=Direct Light Power}
  NamedRerouteDeclaration(None=Multiply(A=Power(Base=@n113, Exp=@n114), B=ScalarParameter() {defaultValue=2, parameterName=Direct Light Scale})) {name=Direct Light, nodeColor=<rgba 1 0 0.0469 1>}
  n115 = WorldPosition()
  n116 = Custom() {code="GetPrimitiveData(Parameters).InstanceLocalBoundsCenter.xyz", description=BoundsCenter}
  n117 = Custom() {code="GetPrimitiveData(Parameters).InstanceLocalBoundsExtent.xyz", description=BoundsExtent}
  n118 = Divide(
   A=Subtract(
    A=TransformPosition(None=@n115.XYZ) {transformSourceType=TRANSFORMPOSSOURCE_World, transformType=TRANSFORMPOSSOURCE_Instance}
    B=Subtract(A=@n116, B=@n117)
   )
   B=Multiply(A=@n117) {constB=2}
  )
  n119 = RotateAboutAxis(
   NormalizedRotationAxis=Constant3Vector() {constant=<rgba 0 1 0 0>}
   RotationAngle=Constant() {r=-0.05}
   PivotPoint=Constant3Vector()
   Position=@n118
  )
  NamedRerouteDeclaration(None=MaterialFunctionCall(Float3=Add(A=@n118, B=@n119)) {materialFunction=&BreakOutFloat3Components}.B) {name=Rotated Bound Z, nodeColor=<rgba 1 0.4453 0 1>}
  NamedRerouteDeclaration(None=@n104.World position offset) {name=Sprite WPO, nodeColor=<rgba 0 1 0.2969 1>}
  NamedRerouteDeclaration(None=Transform(None=Normalize(VectorInput=@n54)) {transformSourceType=TRANSFORMSOURCE_World, transformType=TRANSFORM_Local}) {name=Vector, nodeColor=<rgba 1 0.2578 0 1>}
  n120 = VectorParameter() {parameterName=Param_7, defaultValue=<rgba 0 0.8 1 1>}
  n121 = Multiply(A=ScalarParameter() {defaultValue=0.1, parameterName=Param_30}, B=@n120.RGB)
  n122 = AtmosphericLightVector()
  n123 = Power(Base=Saturate(None=DotProduct(A=@n71, B=@n122)), Exp=ScalarParameter() {defaultValue=1, parameterName=Dot Power})
  NamedRerouteDeclaration(None=VertexInterpolator(VS=LinearInterpolate(A=@n121, B=@n120.RGB, Alpha=@n123)).PS) {name=asdf, nodeColor=<rgba 0.3828 1 0 1>}
  n124 = MaterialFunctionCall(Direction=NamedRerouteUsage(), SH=@n64) {materialFunction=&MF_SHDecode}
  NamedRerouteDeclaration(None=VertexInterpolator(VS=OneMinus(None=@n124.Result)).PS) {name=GI, nodeColor=<rgba 0 0.0625 1 1>}
  n125 = NamedRerouteUsage()
  n126 = ScalarParameter() {defaultValue=9, parameterName=GI Power}
  n127 = Multiply(A=DotProduct(A=CameraVectorWS(), B=SkyAtmosphereLightDirection())) {constB=-1}
  n128 = NamedRerouteUsage()
  n129 = ScalarParameter() {defaultValue=8, parameterName=Transmittance Power}
  n130 = MaterialFunctionCall(
   Color_Morning=VectorParameter() {parameterName=Transmittance Color Morning, defaultValue=<rgba 1 0.3707 0.1959 1>}.RGB
   Color_Day=VectorParameter() {parameterName=Transmittance Color, defaultValue=<rgba 1 0.6406 0.5897 1>}.RGB
   Color_Evening=VectorParameter() {parameterName=Transmittance Color Evening, defaultValue=<rgba 1.25 0.4675 0.7874 1>}.RGB
  ) {materialFunction=&Color_Time_of_Day}
  n131 = ScalarParameter() {defaultValue=10, parameterName=Param_32}
  n132 = Multiply(
   A=Power(Base=Saturate(None=@n127), Exp=ScalarParameter() {defaultValue=5, parameterName=Transmitance_AreaPower})
   B=Multiply(A=Power(Base=@n128, Exp=@n129), B=Multiply(A=@n130.Result, B=@n131))
  )
  n133 = MaterialFunctionCall(
   Color_Morning=VectorParameter() {parameterName=Color Morning, defaultValue=<rgba 0.035 1.3896 1.75 1>}.RGB
   Color_Day=VectorParameter() {parameterName=Color Day, defaultValue=<rgba 0 1.9583 2.5 1>}.RGB
   Color_Evening=VectorParameter() {parameterName=Color Evening, defaultValue=<rgba 0 2.1542 2.75 1>}.RGB
   Color_Night=VectorParameter() {parameterName=Color NIght, defaultValue=<rgba 0 0.9 1.8 1>}.RGB
  ) {materialFunction=&Color_Time_of_Day}
  n134 = MaterialFunctionCall(
   Value_Morning=ScalarParameter() {defaultValue=0.17, parameterName=Shadow Scale Morning}
   Value_Day=ScalarParameter() {defaultValue=0.1, parameterName=Shadow Scale Day}
   Value_Evening=ScalarParameter() {defaultValue=0.05, parameterName=Shadow Scale Evening}
   Value_Night=ScalarParameter() {defaultValue=0.09, parameterName=Shadow Scale Night}
  ) {materialFunction=&Scalar_Time_of_Day}
  n135 = NamedRerouteUsage()
  n136 = Multiply(
   A=Power(Base=@n135, Exp=ScalarParameter() {defaultValue=1, parameterName=SH Overall Power})
   B=ScalarParameter() {defaultValue=1, parameterName=SH Scale}
  )
  n137 = Power(Base=@n135, Exp=ScalarParameter() {defaultValue=4, parameterName=Additional SH Power})
  n138 = ComponentMask(None=CollectionParameter() {collection=&FortniteMaterialParameters, parameterName=SunLightColor}) {b=true, g=true, r=true}
  n139 = Add(
   A=LinearInterpolate(A=Multiply(A=@n133.Result, B=@n134.Result), B=@n133.Result, Alpha=Saturate(None=@n136))
   B=Multiply(A=Multiply(A=@n137, B=@n138), B=ScalarParameter() {defaultValue=1, parameterName=Sh Scale})
  )
  n140 = ScalarParameter() {defaultValue=0.6941, parameterName=Shadow Darken}
  n141 = NamedRerouteUsage()
  n142 = Add(
   A=Multiply(A=Power(Base=@n125, Exp=@n126), B=ScalarParameter() {defaultValue=0.55, parameterName=GI Scale})
   B=Multiply(A=Add(A=@n132, B=@n139), B=LinearInterpolate(A=@n140, Alpha=@n141))
  )
  n143 = ScalarParameter() {defaultValue=0.1, parameterName=Hue Shift Percent}
  n144 = OneMinus(None=NamedRerouteUsage())
  n145 = ScalarParameter() {defaultValue=6, parameterName=Power SH Hue Shift}
  NamedRerouteDeclaration(None=LinearInterpolate(
    A=@n142
    B=MaterialFunctionCall(Hue Shift Percentage=@n143, Texture=@n142) {materialFunction=&HueShift}.Result
    Alpha=Power(Base=@n144, Exp=@n145)
   )) {name=Emissive, nodeColor=<rgba 0 0.0156 1 1>}
  n146 = Add(A=WorldPosition() {worldPositionShaderOffset=WPT_ExcludeAllShaderOffsets}.XYZ, B=NamedRerouteUsage())
  n147 = NamedRerouteUsage()
  n148 = Custom() {code="GetPrimitiveData(Parameters).InstanceLocalBoundsExtent.xyz", description=BoundsExtent}
  n149 = Subtract(
   A=TransformPosition(None=Add(A=@n146, B=@n147)) {transformSourceType=TRANSFORMPOSSOURCE_World, transformType=TRANSFORMPOSSOURCE_Instance}
   B=Subtract(A=Custom() {code="GetPrimitiveData(Parameters).InstanceLocalBoundsCenter.xyz", description=BoundsCenter}, B=@n148)
  )
  NamedRerouteDeclaration(None=Divide(A=@n149, B=Multiply(A=@n148) {constB=2})) {name=Transformed BBX, nodeColor=<rgba 0 0.2031 1 1>}
  NamedRerouteDeclaration(None=Add(A=Add(A=@n104.World position offset, B=@n105), B=NamedRerouteUsage())) {name=WPO, nodeColor=<rgba 0 1 0.1797 1>}
  NamedRerouteDeclaration(None=@n93.G) {name=Noise, nodeColor=<rgba 0 0.7422 1 1>}
  n150 = Time()
  n151 = VectorParameter() {parameterName=Param_8, defaultValue=<rgba -0.25 -0.125 0 0>}
  n152 = Divide(A=Add(A=LocalPosition().XYZ, B=Multiply(A=@n150, B=@n151.RGB)), B=ScalarParameter() {defaultValue=6, parameterName=Param_31}) {constB=200}
  n153 = TextureObject() {samplerType=SAMPLERTYPE_LinearColor, texture=&T_LowAltitudeClouds_N}
  n154 = Constant() {r=8}
  n155 = ScalarParameter() {defaultValue=0.9375, parameterName=AtlasPadding, primitiveDataIndex=8}
  NamedRerouteDeclaration(None=Add(A=MaterialFunctionCall(UVs=@n152, Slices Texture=@n153, Atlas Size=@n154, Atlas Padding=@n155) {materialFunction=&MF_3DVolumeFrom2DSlices}.RGB) {constB=-0.5}) {name=Noise_1, nodeColor=<rgba 0.0312 0 1 1>}
  n156 = Add(A=LocalPosition().XYZ, B=VectorParameter() {parameterName=Offset, defaultValue=<rgba 500 0 113.3061 1>}.RGB)
  n157 = VectorParameter() {parameterName=Bounding Box Size, defaultValue=<rgba 700 700 500 1>}
  NamedRerouteDeclaration(None=Transform(None=Divide(A=@n156, B=@n157.RGB)) {transformSourceType=TRANSFORMSOURCE_Local}) {name=Local To World, nodeColor=<rgba 0 1 0.1797 1>}
  NamedRerouteDeclaration(None=@n149) {name=Non Normalized Bounds, nodeColor=<rgba 0.1016 1 0 1>}
  n158 = LinearInterpolate(
   A=VectorParameter() {parameterName=Param_3, defaultValue=<rgba 0 5 0 0>}.RGB
   B=VectorParameter() {parameterName=Param_2, defaultValue=<rgba 0 -5 0 0>}.RGB
   Alpha=@n29
  )
  n159 = MaterialFunctionCall() {materialFunction=&BoundingBoxBased_0-1_UVW}
  n160 = Multiply(A=OneMinus(None=@n159.R), B=@n159.R)
  n161 = ScalarParameter() {defaultValue=2, parameterName=Param_6}
  NamedRerouteDeclaration(None=LinearInterpolate(B=Transform(None=@n158) {transformSourceType=TRANSFORMSOURCE_Local}, Alpha=Power(Base=@n160, Exp=@n161))) {name=Bend, nodeColor=<rgba 0 0.9531 1 1>}
  n162 = VertexColor()
  n163 = Multiply(A=MaterialFunctionCall(X=@n162.G, Y=@n162.G, Z=@n162.G) {materialFunction=&MakeFloat3}.Result)
  n164 = Multiply(
   A=TextureCoordinate()
   B=LinearInterpolate(B=ScalarParameter() {defaultValue=0.25, parameterName=Param_9}, Alpha=VertexColor().G) {constA=0.2}
  )
  n165 = LinearInterpolate(
   A=TextureSample(UVs=@n164) {texture=/Game/Effects/Fort_Effects/GameplayPulginsAssets/Doom/DoomCastle/Smoke/MaterialAndMesh/Smoke01_D}.G
   B=OneMinus(None=TextureSample(UVs=@n164) {samplerType=SAMPLERTYPE_LinearColor, texture=/Game/Global/Textures/Default/Tileable_Patterns/Cloud_Diffuse}.G)
  )
  n166 = ScalarParameter() {defaultValue=0.5, parameterName=Param_4}
  MaterialExpressionAdd_12 = Add(
   A=AppendVector(A=Add(A=@n162.R), B=Add(A=@n163) {constB=-0.5})
   B=Power(Base=Multiply(A=@n165, B=@n166), Exp=ScalarParameter() {defaultValue=4, parameterName=Param_8})
  )
  MaterialExpressionTransform_0 = Transform(None=Normalize(VectorInput=AtmosphericLightVector())) {transformSourceType=TRANSFORMSOURCE_World, transformType=TRANSFORM_Local}
  n167 = Custom!(SH=@MaterialExpressionAdd_12, Direction=@MaterialExpressionTransform_0) {
   code="//SH decode,  Direction = LightDir, SH = 2band coefficients. (encoded in Y-up)\r\nDirection = normalize(Direction);\r\nfloat4 shEvalDirection;\r\nshEvalDirection.x = 0.282095;\r\nshEvalDirection.y =0.488603 * Direction.z;\r\nshEvalDirection.z = 0.488603 * Direction.y;\r\nshEvalDirection.w =-0.488603 * Direction.x;\r\nreturn saturate(dot(SH, shEvalDirection));"
   description=SphericalHarmonicsEval
   outputType=CMOT_Float1
   inputs [
    {
     inputName=SH
     input {expression=&self:MaterialExpressionAdd_12}
    }
    {
     inputName=Direction
     input {expression=&self:MaterialExpressionTransform_0}
    }
   ]
  }
  n168 = ScalarParameter() {parameterName=SHContrast_LowValue}
  n169 = MaterialFunctionCall(
   Value_Morning=ScalarParameter() {defaultValue=1, parameterName=SHContrast_HighValue_ Morning}
   Value_Day=ScalarParameter() {defaultValue=0.65, parameterName=SHContrast_HighValue Day}
   Value_Evening=ScalarParameter() {defaultValue=0.75, parameterName=SHContrast_HighValue Evening}
   Value_Night=ScalarParameter() {defaultValue=0.65, parameterName=SHContrast_HighValue Night}
  ) {materialFunction=&Scalar_Time_of_Day}
  n170 = Constant()
  n171 = Constant() {r=1}
  n172 = Saturate(None=OneMinus(None=MaterialFunctionCall(Input=@n167, Input Low=@n168, Input High=@n169.Result, Target Low=@n170, Target High=@n171) {materialFunction=&RemapValueRange}.Result))
  NamedRerouteDeclaration(None=@n172) {name=SH, nodeColor=<rgba 0.8984 1 0 1>}
  n173 = MaterialFunctionCall(Float3=Multiply(
    A=ActorPositionWS()
    B=MaterialFunctionCall() {materialFunction=&ObjectScale}.Scale XYZ
   )) {materialFunction=&BreakOutFloat3Components}
  n174 = Multiply(A=Add(A=Add(A=@n173.R, B=@n173.R), B=@n173.B), B=ScalarParameter() {defaultValue=415613, parameterName=Param_11})
  NamedRerouteDeclaration(None=Saturate(None=Frac(None=@n174))) {name=Random WP, nodeColor=<rgba 0.3594 0 1 1>}
  n175 = Length(None=Subtract(A=TextureCoordinate()) {constB=0.5}) {is=class MaterialExpressionLength}
  n176 = ScalarParameter() {defaultValue=3, parameterName=Param_35}
  NamedRerouteDeclaration(None=Saturate(None=Multiply(A=@n175, B=@n176))) {name=UV Mask, nodeColor=<rgba 0 1 0.8594 1>}
  NamedRerouteDeclaration(None=MaterialFunctionCall(
    Color_Morning=VectorParameter() {parameterName="Color_Morning (V3)", defaultValue=<rgba 0.187 1.1362 1.3987 1>}.RGB
    Color_Day=VectorParameter() {parameterName="Color_Day (V3)", defaultValue=<rgba 0.0482 0.7938 1 1>}.RGB
    Color_Evening=VectorParameter() {parameterName="Color_Evening (V3)", defaultValue=<rgba 0.0631 0.2331 0.3906 1>}.RGB
    Color_Night=VectorParameter() {parameterName="Color_Night (V3)", defaultValue=<rgba 0.0712 0.3034 1 1>}.RGB
   ) {materialFunction=&Color_Time_of_Day}.Result) {name=TOD, nodeColor=<rgba 1 0.9141 0 1>}
  NamedRerouteUsage()
  NamedRerouteUsage()
  NamedRerouteUsage()
  NamedRerouteUsage()
  NamedRerouteUsage()
  NamedRerouteUsage()
  NamedRerouteUsage()
  NamedRerouteUsage()
  NamedRerouteUsage()
  NamedRerouteUsage()
  NamedRerouteUsage()
  NamedRerouteUsage()
  NamedRerouteUsage()
  NamedRerouteUsage()
  NamedRerouteUsage()
  NamedRerouteUsage()
  NamedRerouteUsage()
  n177 = NamedRerouteUsage()
  MaterialExpressionMultiply_27 = Multiply(A=@n177, B=@n177)
  n178 = Convert(None=@MaterialExpressionMultiply_27) {
   nodeName=Break Float3
   convertInputs [{
     type=Vector3
     expressionInput {
      expression=&self:MaterialExpressionMultiply_27
      inputName=None
      mask=0
      maskA=0
      maskB=0
      maskG=0
      maskR=0
      outputIndex=0
     }
    }]
   convertOutputs [{type=Scalar}, {type=Scalar}, {type=Scalar}]
  }
  n179 = Add(A=@n178, B=@n178)
  n180 = OneMinus(None=Saturate(None=Add(A=@n179, B=@n178)))
  OneMinus(None=Multiply(A=@n180, B=@n180))
  OneMinus()
  Power(Base=@n172, Exp=ScalarParameter() {defaultValue=2, parameterName=SH Power})
  n181 = DotProduct(A=CameraVectorWS(), B=@n27)
  Saturate(None=Multiply(A=Saturate(None=@n181), B=ScalarParameter() {defaultValue=1, parameterName=Scale}))
  Saturate()
  n182 = MaterialFunctionCall() {materialFunction=/Game/Effects/Niagara/MasterMaterials/Functions/InstanceBounds_0-1_UVW}
  Saturate(None=DotProduct(A=ConstantBiasScale(None=@n182.RGB) {bias=-0.5, scale=2}, B=AtmosphericLightVector()))
  Saturate()
  ScalarParameter() {defaultValue=0.148, parameterName=Param_28}
  ScalarParameter() {defaultValue=9236.9199, parameterName=Param_27}
  n183 = ScalarParameter() {defaultValue=1951094.75, parameterName=Param_14}
  Subtract(B=@n183)
  Subtract(B=@n183)
  Subtract(B=CameraVectorWS())
  TextureCoordinate()
  TextureCoordinate()
  Transform(None=@n156) {transformSourceType=TRANSFORMSOURCE_Local}
  Transform() {transformSourceType=TRANSFORMSOURCE_World, transformType=TRANSFORM_Instance}
  Transform(None=@n70) {transformSourceType=TRANSFORMSOURCE_World, transformType=TRANSFORM_Local}
  VertexInterpolator()
  VertexInterpolator()
  WorldPosition() {worldPositionShaderOffset=WPT_CameraRelative}
  WorldPosition() {worldPositionShaderOffset=WPT_CameraRelative}
 ]
}
