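# Unlit translucent two-sided material for geothermal volume smoke in the BugHut effects, animating looping smoke cards within the shader.
# UV1 per-card data drives sprite scaling, flipbook rotation, fades and camera-offset sort order, alongside world-force drift.
# An 8x8 flipbook and panning world-space noise build eroded, dithered opacity, with a step mask keyed to quality and shading path.
# Fake spherical sun shading and time-of-day albedo form the color, desaturated by distance on the default path.
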Material /Game/Effects/Fort_Effects/Effects/Environment/Bugs/BugHut/Materials/M_GeoThermal_VolumeSmoke_LOD
graph {
 properties {
  bApplyCloudFogging=true
  bAutomaticallySetUsageInEditor=false
  bCastRayTracedShadows=false
  blendMode=BLEND_Translucent
  floatPrecisionMode=MFPM_Full
  pixelDepthOffsetMode=PDOM_Legacy
  shadingModel=MSM_Unlit
  translucencyLightingMode=TLM_VolumetricPerVertexDirectional
  twoSided=true
  usedShadingModels=Unlit
 }
 nodes [
  Add()
  n1 = ScalarParameter() {defaultValue=2350, parameterName=min scale}
  n2 = ScalarParameter() {defaultValue=2750, parameterName=Max Scale}
  n3 = ComponentMask(None=TextureCoordinate() {coordinateIndex=1}) {g=true}
  n4 = ScalarParameter() {defaultValue=0.5, parameterName=Initial Scale}
  n5 = ScalarParameter() {defaultValue=1.5, parameterName=Final Scale}
  n6 = ComponentMask(None=TextureCoordinate() {coordinateIndex=1}) {r=true}
  n7 = ScalarParameter() {defaultValue=1.5, parameterName=Smoke Taper}
  n8 = Multiply(
   A=LinearInterpolate(
    A=ScalarParameter() {parameterName=Param_5}
    B=ScalarParameter() {defaultValue=1.2948, parameterName=Param_4}
    Alpha=MaterialFunctionCall() {materialFunction=/Game/Effects/Niagara/MasterMaterials/Functions/InstanceBounds_0-1_UVW}.R
   )
   B=SphereMask(
    A=Power(Base=@n6, Exp=@n7)
    B=ScalarParameter() {defaultValue=1, parameterName="Scale  Fade Start"}
    Radius=ScalarParameter() {defaultValue=0.5387, parameterName=Scale fade Duration}
    Hardness=ScalarParameter() {parameterName="Scale  Fade Hardness"}
   )
  )
  n9 = Multiply(A=NamedRerouteUsage(), B=Multiply(A=LinearInterpolate(A=@n1, B=@n2, Alpha=@n3), B=LinearInterpolate(A=@n4, B=@n5, Alpha=@n8)))
  n10 = MaterialFunctionCall!(
   Center Location=WorldPosition() {worldPositionShaderOffset=WPT_CameraRelative}.XYZ
   0-1 UVs=TextureCoordinate()
   XY scale=@n9
   OverridePivot=StaticBoolParameter() {parameterName="OverridePivot (SB)"}
   "Normals (Optional)"
  ) {materialFunction=&Sprite}
  n11 = VectorParameter() {parameterName=World Force, defaultValue=<rgba 1500 0 150 0>}
  n12 = Add(
   A=@n10.World position offset
   B=Multiply(A=Transform(None=@n11.RGB) {transformSourceType=TRANSFORMSOURCE_World, transformType=TRANSFORM_Local}, B=NamedRerouteUsage())
  )
  Add(A=@n12, B=NamedRerouteUsage())
  n13 = DotProduct(
   A=Transform(None=AtmosphericLightVector()) {transformSourceType=TRANSFORMSOURCE_World, transformType=TRANSFORM_Tangent}
   B=TextureSample(UVs=NamedRerouteUsage()) {samplerType=SAMPLERTYPE_Normal, texture=/Game/Effects/Niagara/MasterMaterials/Smoke/Textures/NEW/RT4x4_3_Tex}.RGB
  )
  Add(B=Multiply(A=Power(Base=@n13) {constExponent=4}, B=ScalarParameter() {defaultValue=0.5, parameterName=Param_6}))
  n14 = TextureSample(UVs=NamedRerouteUsage()) {samplerType=SAMPLERTYPE_LinearColor, texture=&/CA_SmokeFadeIn_Out}
  n15 = Multiply(A=NamedRerouteUsage(), B=@n14.A)
  n16 = SphereMask(
   A=ComponentMask(None=TextureCoordinate() {coordinateIndex=1}) {r=true}
   B=ScalarParameter() {defaultValue=0.9399, parameterName=Length Fade}
   Radius=ScalarParameter() {defaultValue=1.6802, parameterName=Length Fade Radius}
   Hardness=ScalarParameter() {defaultValue=0.1067, parameterName=Length Fade Hardness}
  )
  n17 = ScalarParameter() {defaultValue=1, parameterName=Step}
  n18 = ScalarParameter() {parameterName=Step Low}
  n19 = Step(
   Y=ComponentMask(None=TextureCoordinate() {coordinateIndex=1}) {r=true}
   X=Multiply(
    A=QualitySwitch(Default=@n17, Low=@n18, High=@n17, Medium=@n18, Epic=@n17)
    B=ShadingPathSwitch(Default=@n17, Deferred=@n17, Forward=@n18, Mobile=@n18)
   )
  )
  n20 = Multiply(A=Multiply(A=Multiply(A=@n15, B=@n16), B=@n19), B=ScalarParameter() {defaultValue=0.5, parameterName=Overall Opacity}) {constA=1}
  n21 = Multiply(A=@n20, B=ScalarParameter() {defaultValue=1.5, parameterName=Param_9})
  n22 = Multiply(A=@n21, B=Multiply(A=NamedRerouteUsage(), B=ScalarParameter() {defaultValue=1, parameterName=Fine Erode Scale}))
  n23 = Multiply(A=NamedRerouteUsage(), B=ScalarParameter() {defaultValue=1, parameterName=Erode Scale})
  n24 = LinearInterpolate(
   B=ScalarParameter() {defaultValue=1.25, parameterName=Threshold Boost}
   Alpha=ShadingPathSwitch(Default=Multiply(A=@n22, B=@n23), Forward=@n21, Mobile=@n21)
  )
  MaterialFunctionCall!(Alpha Threshold=@n24, Random) {materialFunction=/Game/Environments/Asteria/Sets/NeonCity/Techart/Screen/MaterialFunctions/DitherTemporalAABlue}
  n25 = VertexInterpolator(VS=PreSkinnedNormal())
  n26 = Add(A=NamedRerouteUsage(), B=Multiply(A=NamedRerouteUsage(), B=ScalarParameter() {defaultValue=750, parameterName=Param_8}))
  n27 = Subtract(
   A=VertexInterpolator(VS=LocalPosition() {localOrigin=InstancePreSkinning}.XYZ).PS
   B=WorldPosition() {worldPositionShaderOffset=WPT_CameraRelative}.XYZ
  )
  n28 = Multiply(A=NamedRerouteUsage()) {constB=128}
  n29 = Subtract(A=Subtract(A=NamedRerouteUsage(), B=Add(A=@n25.PS, B=@n26)), B=Divide(A=Subtract(A=@n27, B=@n28)) {constB=1000})
  n30 = ScalarParameter() {defaultValue=1, parameterName=Param_10}
  n31 = NamedRerouteUsage()
  n32 = Normalize(VectorInput=Add(A=Transform(None=@n29) {transformSourceType=TRANSFORMSOURCE_Local}, B=Multiply(A=@n30, B=@n31)))
  MaterialFunctionCall(Normal=@n32, Flatness=ScalarParameter() {parameterName=Flatten Normal}) {materialFunction=&FlattenNormal}
  Multiply()
  Multiply(A=OneMinus(None=VertexColor().G), B=@n24)
  n33 = Time()
  n34 = VectorParameter() {defaultValue=<rgba -25 0 25 0>}
  n35 = Multiply(A=Add(A=WorldPosition().XYZ, B=Multiply(A=@n33, B=@n34.RGB)), B=ScalarParameter() {defaultValue=0.001})
  NamedRerouteDeclaration(None=Multiply(A=VectorNoise(World Position=@n35) {noiseFunction=VNF_CurlALU}, B=ScalarParameter() {defaultValue=150, parameterName=Param_1})) {name=noise, nodeColor=<rgba 0.9922 0 1 1>}
  n36 = ScalarParameter() {defaultValue=8, parameterName=SubUV}
  n37 = ComponentMask(None=TextureCoordinate() {coordinateIndex=1}) {g=true}
  n38 = MaterialFunctionCall!(
   U_Images=@n36
   V_Images=@n36
   SubUV Texture RGB
   Time Speed=ScalarParameter() {defaultValue=1, parameterName=SubUV Time Scale}
   Manual Time Control=StaticBoolParameter() {defaultValue=true}
   Time=NamedRerouteUsage()
   UVs=Rotator(Coordinate=TextureCoordinate(), Time=LinearInterpolate(Alpha=@n37) {constA=-360, constB=360})
  ) {materialFunction=/Game/Effects/Fort_Effects/Materials/_Functions/MF_SubUV}
  NamedRerouteDeclaration(None=@n38.UVs) {name=UV, nodeColor=<rgba 0 1 0.25 1>}
  n39 = TextureCoordinate() {coordinateIndex=1}
  n40 = MaterialFunctionCall(
   Offset Amount=LinearInterpolate(B=ScalarParameter() {defaultValue=750, parameterName=Indexed Sort Order}, Alpha=ComponentMask(None=@n39) {r=true})
   World Position=WorldPosition().XYZ
  ) {materialFunction=/Engine/Functions/Engine_MaterialFunctions02/WorldPositionOffset/CameraOffset}
  n41 = Divide(A=PixelDepth().R, B=ScalarParameter() {defaultValue=1500, parameterName=Shrink over Distance})
  NamedRerouteDeclaration(None=Multiply(A=Add(A=@n12, B=@n40.Camera Offset), B=Saturate(None=@n41))) {name=WPO, nodeColor=<rgba 0.2891 1 0 1>}
  n42 = AtmosphericLightVector()
  n43 = OneMinus(None=Saturate(None=DotProduct(A=@n32, B=@n42)))
  n44 = MaterialFunctionCall(
   Value_Morning=ScalarParameter() {defaultValue=2, parameterName=Power MOrning}
   Value_Day=ScalarParameter() {defaultValue=1, parameterName=Power Day}
   Value_Evening=ScalarParameter() {defaultValue=2, parameterName=Power Evening}
   Value_Night=ScalarParameter() {defaultValue=4, parameterName=Power Night}
  ) {materialFunction=&Scalar_Time_of_Day}
  n45 = OneMinus(None=VertexColor().G)
  n46 = CameraVectorWS()
  n47 = Saturate(None=DotProduct(A=Multiply(A=@n46) {constB=-1}, B=SkyAtmosphereLightDirection()))
  n48 = LinearInterpolate(B=ScalarParameter() {defaultValue=4, parameterName=Edge falloff power}, Alpha=Multiply(A=@n47, B=@n47)) {constA=1, constB=4}
  NamedRerouteDeclaration(None=Multiply(A=Power(Base=@n43, Exp=@n44.Result), B=Power(Base=@n45, Exp=@n48))) {name=Spherical Shading, nodeColor=<rgba 1 0.1406 0 1>}
  NamedRerouteDeclaration(None=@n20) {name=Opacity, nodeColor=<rgba 0 0.5547 1 1>}
  NamedRerouteDeclaration(None=@n9) {name=Size, nodeColor=<rgba 0 0.0625 1 1>}
  n49 = WorldPosition()
  n50 = Multiply(A=Add(A=Time(), B=NamedRerouteUsage()), B=VectorParameter() {parameterName=Noise Pan Dir, defaultValue=<rgba 500 0 -50 0>}.RGB)
  n51 = TextureSample(UVs=Multiply(A=Add(A=@n49.XYZ, B=@n50), B=ScalarParameter() {defaultValue=0.0005, parameterName=Noise Scale})) {samplerType=SAMPLERTYPE_LinearColor, texture=/Game/Global/MaterialFunctions/VT_ClothSimWPO}
  NamedRerouteDeclaration(None=@n51.RGB) {name=noise top, nodeColor=<rgba 0.8516 0 1 1>}
  n52 = Add(A=NamedRerouteUsage(), B=Multiply(A=LocalPosition().XYZ) {constB=-1})
  n53 = Transform(None=WorldPosition() {worldPositionShaderOffset=WPT_CameraRelative}.XYZ) {transformSourceType=TRANSFORMSOURCE_World, transformType=TRANSFORM_Local}
  NamedRerouteDeclaration(None=Normalize(VectorInput=Add(A=@n52, B=@n53))) {name=Card Vector, nodeColor=<rgba 1 0.6797 0 1>}
  n54 = NamedRerouteUsage()
  n55 = NamedRerouteUsage()
  NamedRerouteDeclaration(None=Normalize(VectorInput=CrossProduct(A=@n54, B=@n55))) {name=Waving Tangent, nodeColor=<rgba 0.9688 0 1 1>}
  n56 = TextureObjectParameter(UVs=@n38.UVs) {parameterName=TextureObject, samplerType=SAMPLERTYPE_LinearColor, texture=/Game/Effects/Fort_Effects/Textures/Smoke/T_FallingMist_8x8_01}
  NamedRerouteDeclaration(None=TextureSample(UVs=@n38.UVs, Tex=@n56).G) {name=Texture, nodeColor=<rgba 0 0.7891 1 1>}
  n57 = Time()
  n58 = NamedRerouteUsage()
  n59 = Add(
   A=ComponentMask(None=TextureCoordinate() {coordinateIndex=1}) {r=true}
   B=Multiply(A=Add(A=@n57, B=@n58), B=ScalarParameter() {defaultValue=0.01, parameterName=Delta Time Scale})
  )
  n60 = ScalarParameter() {defaultValue=1, parameterName=Loop Duration}
  NamedRerouteDeclaration(None=Divide(A=Fmod(A=@n59, B=@n60), B=@n60)) {name=Normalized AGe, nodeColor=<rgba 0 0.6484 1 1>}
  n61 = VectorParameter() {parameterName=Albedo, defaultValue=<rgba 2.924 1.0204 1.7412 1>}
  n62 = MaterialFunctionCall(
   Color_Morning=@n61.RGB
   Color_Day=@n61.RGB
   Color_Evening=@n61.RGB
   Color_Night=VectorParameter() {parameterName=Albedo Night, defaultValue=<rgba 0 0.5 1 1>}.RGB
  ) {materialFunction=&Color_Time_of_Day}
  NamedRerouteDeclaration(None=@n62.Result) {name=Albedo, nodeColor=<rgba 0 0.5547 1 1>}
  NamedRerouteDeclaration(None=@n10.Normals) {name=WPO Normals, nodeColor=<rgba 0.6641 0 1 1>}
  NamedRerouteDeclaration(None=@n19) {name=Step, nodeColor=<rgba 0.1953 0 1 1>}
  n63 = ComponentMask(None=TextureCoordinate()) {g=true}
  n64 = ScalarParameter() {defaultValue=2, parameterName=Param_7}
  NamedRerouteDeclaration(None=LinearInterpolate(A=Power(Base=@n63, Exp=@n64), Alpha=@n51.G)) {name=Erode, nodeColor=<rgba 1 0 0.3984 1>}
  NamedRerouteDeclaration() {name=Random Offset based on WP, nodeColor=<rgba 0 0.2031 1 1>}
  n65 = DotProduct(A=VectorParameter() {parameterName=Param_1, defaultValue=<rgba 0 0 1 0>}.RGB, B=CameraVectorWS())
  NamedRerouteDeclaration(None=Multiply(A=Saturate(None=@n65), B=NamedRerouteUsage())) {name=Top Brighten, nodeColor=<rgba 0 1 0.9063 1>}
  n66 = ComponentMask(None=TextureCoordinate()) {g=true}
  n67 = ScalarParameter() {defaultValue=0.5, parameterName=Fine Erode Length}
  n68 = Add(A=Time(), B=NamedRerouteUsage())
  n69 = VectorParameter() {parameterName=Noise Pan Dir, defaultValue=<rgba 500 0 -50 0>}
  n70 = Multiply(A=Add(A=WorldPosition().XYZ, B=Multiply(A=@n68, B=@n69.RGB)), B=ScalarParameter() {defaultValue=0.001, parameterName=Fine Noise Scale})
  NamedRerouteDeclaration(None=LinearInterpolate(
    A=Power(Base=@n66, Exp=@n67)
    Alpha=TextureSample(UVs=@n70) {samplerType=SAMPLERTYPE_LinearColor, texture=&VT_ClothSimWPO}.G
   )) {name=Erode Fine, nodeColor=<rgba 1 0 0.3984 1>}
  NamedRerouteUsage()
  n71 = Saturate(None=Divide(A=PixelDepth().R, B=ScalarParameter() {defaultValue=125000, parameterName=Desat Distance}))
  n72 = NamedRerouteUsage()
  n73 = ScalarParameter() {defaultValue=0.1, parameterName=Top Brighten}
  n74 = NamedRerouteUsage()
  n75 = ScalarParameter() {defaultValue=1, parameterName=Param_2}
  n76 = Multiply(A=Multiply(A=Power(Base=@n74, Exp=@n75), B=ScalarParameter() {defaultValue=0.25, parameterName=Param_3}), B=NamedRerouteUsage())
  n77 = LinearInterpolate(A=Multiply(A=@n62.Result, B=ScalarParameter() {defaultValue=0.0035, parameterName=Darken Albedo}), B=@n62.Result, Alpha=NamedRerouteUsage())
  n78 = Constant() {r=1}
  n79 = Multiply(
   A=Add(A=Multiply(A=@n72, B=@n73), B=Add(A=@n76, B=@n77))
   B=MaterialFunctionCall(Value_Morning=@n78, Value_Day=@n78, Value_Evening=@n78, Value_Night=Constant() {r=0.75}) {materialFunction=&Scalar_Time_of_Day}.Result
  )
  ShadingPathSwitch(Default=Desaturation(None=@n79, Fraction=LinearInterpolate(Alpha=@n71) {constB=0.15}), Forward=@n79, Mobile=@n79)
  VectorParameter() {parameterName=Albedo, defaultValue=<rgba 2.924 1.0204 1.7412 1>}
  n80 = NamedRerouteUsage()
  n81 = CrossProduct(A=@n80, B=NamedRerouteUsage())
  n82 = Normalize(VectorInput=@n81)
  n83 = NamedRerouteUsage()
  n84 = LocalPosition()
  n85 = Multiply(A=NamedRerouteUsage(), B=VertexColor().R)
  n86 = ScalarParameter() {defaultValue=-100, parameterName=Param_16}
  n87 = Add(
   A=Multiply(A=Multiply(A=@n82, B=@n83) {constB=0.25}, B=ComponentMask(None=@n84.XYZ) {r=true})
   B=Multiply(A=Multiply(A=@n85, B=@n86) {constB=-100}, B=NamedRerouteUsage())
  )
  n88 = OneMinus(None=@n83)
  n89 = CrossProduct(A=@n81, B=@n80)
  n90 = CrossProduct(A=@n82, B=@n80)
  n91 = ComponentMask(None=@n84.XYZ) {g=true}
  n92 = Multiply(A=Add(A=Multiply(A=@n88, B=@n89), B=Multiply(A=@n90, B=@n83) {constB=0.25}), B=@n91)
  n93 = Multiply(A=Subtract(A=Add(A=@n87, B=@n92), B=Multiply(A=@n89, B=@n91))) {constB=0.01}
  VertexInterpolator(VS=Divide(
    A=Transform(None=@n93) {transformSourceType=TRANSFORMSOURCE_Local}
    B=MaterialFunctionCall() {materialFunction=&ObjectScale}.Scale XYZ
   ))
  VertexInterpolator(VS=PreSkinnedNormal())
  VertexInterpolator(VS=LocalPosition() {localOrigin=InstancePreSkinning}.XYZ)
  VertexInterpolator(VS=LocalPosition() {localOrigin=InstancePreSkinning}.XYZ)
  VertexInterpolator()
 ]
}
