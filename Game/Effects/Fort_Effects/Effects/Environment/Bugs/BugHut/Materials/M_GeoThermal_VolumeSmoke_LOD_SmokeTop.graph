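# Translucent unlit two-sided material for the billboarded smoke top of the BugHut geothermal plume, judging by its name.
# UV1 values and time drive looping particle-like age in the shader, feeding sprite scaling, world-force drift, camera-offset sorting and near-camera shrinking.
# A randomly rotated mist flipbook supplies opacity with sphere-mask, depth and dither fades.
# A custom 2D raymarch along the sun direction through that flipbook produces time-of-day-scaled self-shadowed lighting and backlight.
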Material /Game/Effects/Fort_Effects/Effects/Environment/Bugs/BugHut/Materials/M_GeoThermal_VolumeSmoke_LOD_SmokeTop
graph {
 properties {
  bAutomaticallySetUsageInEditor=false
  blendMode=BLEND_Translucent
  pixelDepthOffsetMode=PDOM_Legacy
  shadingModel=MSM_Unlit
  translucencyLightingMode=TLM_VolumetricPerVertexDirectional
  twoSided=true
  usedShadingModels=Unlit
 }
 nodes [
  AtmosphericLightVector()
  ComponentMask(None=CollectionParameter() {collection=&FortniteMaterialParameters, parameterName=CloudsShadow}) {b=true, g=true, r=true}
  ComponentMask() {g=true, r=true}
  ConstantBiasScale(None=MaterialFunctionCall() {materialFunction=/Game/Effects/Niagara/MasterMaterials/Functions/InstanceBounds_0-1_UVW}.RGB) {bias=-0.5, scale=2}
  Divide(
   A=MaterialFunctionCall() {materialFunction=&InstanceBounds_0-1_UVW}.RGB
   B=MaterialFunctionCall() {materialFunction=&ObjectScale}.Scale XYZ
  )
  MaterialFunctionCall() {materialFunction=/Game/Effects/Fort_Effects/Materials/_Functions/CurveColor_Time_of_Day_S23}
  MaterialFunctionCall() {materialFunction=&DitherTemporalAA}
  n1 = ScalarParameter() {defaultValue=2350, parameterName=min scale}
  n2 = ScalarParameter() {defaultValue=2750, parameterName=Max Scale}
  n3 = ComponentMask(None=TextureCoordinate() {coordinateIndex=1}) {g=true}
  n4 = ScalarParameter() {defaultValue=0.5, parameterName=Initial Scale}
  n5 = ScalarParameter() {defaultValue=1.5, parameterName=Final Scale}
  n6 = SphereMask(
   A=NamedRerouteUsage()
   B=ScalarParameter() {defaultValue=1, parameterName="Scale  Fade Start"}
   Radius=ScalarParameter() {defaultValue=0.5387, parameterName=Scale fade Duration}
   Hardness=ScalarParameter() {parameterName="Scale  Fade Hardness"}
  )
  n7 = MaterialFunctionCall!(
   Center Location=WorldPosition().XYZ
   0-1 UVs=TextureCoordinate()
   XY scale=Multiply(A=LinearInterpolate(A=@n1, B=@n2, Alpha=@n3), B=LinearInterpolate(A=@n4, B=@n5, Alpha=@n6))
   OverridePivot=StaticBoolParameter() {parameterName="OverridePivot (SB)"}
   "Normals (Optional)"
  ) {materialFunction=&Sprite}
  n8 = ComponentMask(None=TextureCoordinate() {coordinateIndex=1}) {r=true}
  n9 = Multiply(A=Time(), B=ScalarParameter() {defaultValue=0.05, parameterName=Delta Time Scale})
  n10 = ScalarParameter() {defaultValue=1, parameterName=Loop Duration}
  n11 = Divide(A=Fmod(A=Add(A=@n8, B=@n9), B=@n10), B=@n10)
  n12 = Multiply(
   A=Transform(None=VectorParameter() {parameterName=World Force, defaultValue=<rgba 1500 0 150 0>}.RGB) {transformSourceType=TRANSFORMSOURCE_World, transformType=TRANSFORM_Local}
   B=@n11
  )
  n13 = LinearInterpolate(
   B=ScalarParameter() {defaultValue=750, parameterName=Indexed Sort Order}
   Alpha=ComponentMask(None=TextureCoordinate() {coordinateIndex=1}) {r=true}
  )
  n14 = WorldPosition()
  n15 = PixelDepth()
  n16 = ScalarParameter() {defaultValue=1500, parameterName=Shrink over Distance}
  Multiply(
   A=Add(
    A=Add(A=@n7.World position offset, B=@n12)
    B=MaterialFunctionCall(Offset Amount=@n13, World Position=@n14.XYZ) {materialFunction=/Engine/Functions/Engine_MaterialFunctions02/WorldPositionOffset/CameraOffset}.Camera Offset
   )
   B=Saturate(None=Divide(A=@n15.R, B=@n16))
  )
  n17 = TextureCoordinate() {coordinateIndex=1}
  Multiply(A=LinearInterpolate(
    A=ScalarParameter() {defaultValue=9, parameterName=Random Loop Duration Scale min}
    B=ScalarParameter() {defaultValue=10, parameterName=Random Loop Duration Scale Max}
    Alpha=ComponentMask(None=@n17) {g=true}
   ))
  n18 = Normalize(VectorInput=CrossProduct(A=CameraVectorWS(), B=NamedRerouteUsage()))
  n19 = NamedRerouteUsage()
  NamedRerouteDeclaration(None=Saturate(None=DotProduct(A=@n18, B=@n19))) {name=Sun Dir, nodeColor=<rgba 0.2422 0 1 1>}
  NamedRerouteDeclaration(None=LinearInterpolate(
    A=ScalarParameter() {defaultValue=0.15, parameterName=March Dir}
    B=ScalarParameter() {defaultValue=0.5, parameterName=Marc Dir Hi}
    Alpha=NamedRerouteUsage()
   )) {name=March Dir, nodeColor=<rgba 0.0781 1 0 1>}
  NamedRerouteDeclaration(None=@n11) {name=Normalized AGe, nodeColor=<rgba 0 0.6484 1 1>}
  n20 = VectorParameter() {parameterName=Albedo, defaultValue=<rgba 2.9236 1.0202 1.741 1>}
  NamedRerouteDeclaration(None=@n20.RGB) {name=Albedo, nodeColor=<rgba 0 0.5547 1 1>}
  NamedRerouteDeclaration(None=@n7.Normals) {name=WPO Normals, nodeColor=<rgba 0.6641 0 1 1>}
  n21 = Ceil(None=MaterialFunctionCall() {materialFunction=&BoundingBoxBased_0-1_UVW}.RGB)
  n22 = CameraVectorWS()
  n23 = LinearInterpolate(
   A=ScalarParameter() {defaultValue=-1}
   B=ScalarParameter() {defaultValue=1, parameterName=Param_2}
   Alpha=Saturate(None=DotProduct(A=@n21, B=@n22))
  ) {constA=1, constB=-1}
  NamedRerouteDeclaration(None=@n23) {name=Normal, nodeColor=<rgba 0 0.4609 1 1>}
  OneMinus()
  n24 = ScalarParameter() {defaultValue=8, parameterName=SubUV}
  n25 = LinearInterpolate(Alpha=ComponentMask(None=TextureCoordinate() {coordinateIndex=1}) {g=true}) {constA=-360, constB=360}
  MaterialExpressionMaterialFunctionCall_2 = MaterialFunctionCall!(
   U_Images=@n24
   V_Images=@n24
   SubUV Texture RGB
   Time Speed=ScalarParameter() {defaultValue=1, parameterName=SubUV Time Scale}
   Manual Time Control=StaticBoolParameter() {defaultValue=true}
   Time=@n11
   UVs=Rotator(Coordinate=TextureCoordinate(), Time=@n25)
  ) {materialFunction=/Game/Effects/Fort_Effects/Materials/_Functions/MF_SubUV}
  MaterialExpressionTextureObjectParameter_0 = TextureObjectParameter(UVs=@MaterialExpressionMaterialFunctionCall_2.UVs) {parameterName=TextureObject, samplerType=SAMPLERTYPE_LinearColor, texture=/Game/Effects/Fort_Effects/Textures/Smoke/T_FallingMist_8x8_01}
  MaterialExpressionNamedRerouteUsage_1 = NamedRerouteUsage()
  MaterialExpressionScalarParameter_20 = ScalarParameter() {defaultValue=16, parameterName=NumSteps}
  MaterialExpressionScalarParameter_5 = ScalarParameter() {defaultValue=1, parameterName=TL}
  MaterialExpressionMultiply_35 = Multiply(A=ScalarParameter() {defaultValue=0.0075, parameterName=Shaodw Darken}, B=NamedRerouteUsage())
  MaterialExpressionMultiply_8 = Multiply(
   A=ScalarParameter() {defaultValue=1, parameterName=Overall Density}
   B=MaterialFunctionCall(
    Value_Morning=ScalarParameter() {defaultValue=1, parameterName=Density Scale Morning}
    Value_Day=ScalarParameter() {defaultValue=0.75, parameterName=Density Scale Day}
    Value_Evening=ScalarParameter() {defaultValue=0.5, parameterName=Density Scale Eveneing}
    Value_Night=ScalarParameter() {defaultValue=0.5, parameterName=Density Scale Night}
   ) {materialFunction=&Scalar_Time_of_Day}.Result
  )
  n26 = AtmosphericLightVector()
  MaterialExpressionRotator_1 = Rotator(Coordinate=Multiply(A=Transform(None=@n26) {transformSourceType=TRANSFORMSOURCE_World, transformType=TRANSFORM_Tangent}, B=@n23), Time=@n25)
  MaterialExpressionMultiply_17 = Multiply(
   A=ComponentMask(None=CollectionParameter() {collection=&FortniteMaterialParameters, parameterName=SunLightColor}) {b=true, g=true, r=true}
   B=MaterialFunctionCall(
    Value_Morning=ScalarParameter() {defaultValue=0.1, parameterName=TOD Morning}
    Value_Day=ScalarParameter() {defaultValue=0.15, parameterName=TOD Day}
    Value_Evening=ScalarParameter() {defaultValue=0.15, parameterName=TOD Evening}
    Value_Night=ScalarParameter() {defaultValue=0.25, parameterName=TOD Night}
   ) {materialFunction=&Scalar_Time_of_Day}.Result
  )
  MaterialExpressionScalarParameter_30 = ScalarParameter() {defaultValue=8, parameterName=Contrast}
  n27 = Custom!(
   Tex=@MaterialExpressionTextureObjectParameter_0
   marchDistance=@MaterialExpressionNamedRerouteUsage_1
   numSteps=@MaterialExpressionScalarParameter_20
   TL=@MaterialExpressionScalarParameter_5
   LO=@MaterialExpressionMultiply_35
   density=@MaterialExpressionMultiply_8
   UV=@MaterialExpressionMaterialFunctionCall_2.UVs
   marchDir=@MaterialExpressionRotator_1
   LightColor=@MaterialExpressionMultiply_17
   contrast=@MaterialExpressionScalarParameter_30
  ) {
   code="for (float d = 0.0; d < marchDistance; d += marchDistance / numSteps)\r\n{\r\n    UV += (marchDir * d) / numSteps;\r\n    float4 textureSample = Texture2DSampleLevel(Tex, TexSampler, UV.xy, 0) / (numSteps - contrast);\r\n    // Extracting the red channel\r\n    float densityChannel = textureSample.b;\r\n    TL *= saturate(1.0 - densityChannel);\r\n}\r\nfloat3 LI = LightColor * TL.xxx;\r\nLO += LI * density;\r\nreturn LO;"
   description=2D RayMarching
   inputs [
    {
     inputName=Tex
     input {expression=&self:MaterialExpressionTextureObjectParameter_0}
    }
    {
     inputName=marchDistance
     input {expression=&self:MaterialExpressionNamedRerouteUsage_1}
    }
    {
     inputName=numSteps
     input {expression=&self:MaterialExpressionScalarParameter_20}
    }
    {
     inputName=TL
     input {expression=&self:MaterialExpressionScalarParameter_5}
    }
    {
     inputName=LO
     input {expression=&self:MaterialExpressionMultiply_35}
    }
    {
     inputName=density
     input {expression=&self:MaterialExpressionMultiply_8}
    }
    {
     inputName=UV
     input {
      expression=&self:MaterialExpressionMaterialFunctionCall_2
      outputIndex=2
     }
    }
    {
     inputName=marchDir
     input {expression=&self:MaterialExpressionRotator_1}
    }
    {
     inputName=LightColor
     input {expression=&self:MaterialExpressionMultiply_17}
    }
    {
     inputName=contrast
     input {expression=&self:MaterialExpressionScalarParameter_30}
    }
   ]
  }
  n28 = ScalarParameter() {defaultValue=1, parameterName=Backlight Power}
  n29 = Multiply(
   A=Multiply(A=Power(Base=@n27, Exp=@n28), B=ScalarParameter() {defaultValue=15, parameterName=Backlight Scale})
   B=@MaterialExpressionMultiply_17
  )
  Reroute(None=Multiply(A=@n20.RGB, B=Add(A=@n27, B=@n29)))
  n30 = TextureSample(UVs=@MaterialExpressionMaterialFunctionCall_2.UVs, Tex=@MaterialExpressionTextureObjectParameter_0)
  n31 = SphereMask(
   A=NamedRerouteUsage()
   B=ScalarParameter() {defaultValue=0.5587, parameterName=Opacity Fade Start}
   Radius=ScalarParameter() {defaultValue=0.4846, parameterName=Opacity fade Duration}
   Hardness=ScalarParameter() {parameterName=Opacity Fade Hardness}
  )
  n32 = ComponentMask(None=TextureCoordinate() {coordinateIndex=1}) {r=true}
  n33 = ScalarParameter() {defaultValue=1, parameterName=Param_1}
  n34 = Multiply(A=Multiply(A=Multiply(A=@n30.B, B=@n31), B=Step(Y=@n32, X=@n33)), B=ScalarParameter() {defaultValue=0.5, parameterName=Overall Opacity}) {constA=1}
  Reroute(None=DepthFade(Opacity=@n34))
  n35 = MaterialFunctionCall!(Alpha Threshold=@n34, Random=ComponentMask(None=TextureCoordinate() {coordinateIndex=1}) {g=true}) {materialFunction=/Game/TimeOfDay/CloudsAndStorm/Materials/MaterialFunctions/DitherTemporalAABlue}
  n36 = OneMinus(None=NamedRerouteUsage())
  n37 = WorldPosition()
  n38 = Multiply(A=Time(), B=VectorParameter().RGB)
  n39 = TextureSample(UVs=Multiply(A=Add(A=@n37.XYZ, B=@n38), B=ScalarParameter() {defaultValue=0.0008, parameterName=UV Scale})) {samplerType=SAMPLERTYPE_LinearColor, texture=/Game/Global/MaterialFunctions/VT_ClothSimWPO}
  Reroute(None=Multiply(A=LinearInterpolate(B=@n35.Result, Alpha=@n34), B=LinearInterpolate(A=@n36, Alpha=@n39.G)))
  ScalarParameter() {defaultValue=0.0233, parameterName=MarchDistance}
  VectorParameter() {parameterName=Negative Side, defaultValue=<rgba -1 -1 1 0>}
  VertexColor()
  VertexColor()
 ]
}
