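# Translucent, unlit, two-sided smoke-trail material in the BugHut effects, drawn on spline geometry widened by SplineThicken with world-position and camera offsets.
# Opacity combines noise-distorted panning smoke textures with cloud-broken ends, view-axis, depth and near-camera fades.
# Emissive tints time-of-day albedo using ribbon gradients, hue shift and sun-facing shading.
# Platform and quality switches drop cloud, distortion and lighting detail on weaker targets; despite the name, no raymarch loop appears.
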Material /Game/Effects/Fort_Effects/Effects/Environment/Bugs/BugHut/Materials/M_Raymarch_SmokeSpline
graph {
 properties {
  bAutomaticallySetUsageInEditor=false
  bCastRayTracedShadows=false
  blendMode=BLEND_Translucent
  shadingModel=MSM_Unlit
  twoSided=true
  usedShadingModels=Unlit
 }
 nodes [
  CameraVectorWS()
  CameraVectorWS()
  DotProduct(A=Transform(None=NamedRerouteUsage()) {transformType=TRANSFORM_Instance}, B=CameraVectorWS())
  n1 = ComponentMask(None=VertexTangentWS()) {r=true}
  n2 = LinearInterpolate(
   A=ScalarParameter() {defaultValue=0.1, parameterName=Param_11}
   B=ScalarParameter() {defaultValue=1, parameterName=Param_12}
   Alpha=Power(Base=OneMinus(None=@n1), Exp=ScalarParameter() {defaultValue=0.5, parameterName=Param_5})
  )
  Multiply(B=@n2)
  Multiply(B=@n2)
  n3 = MaterialFunctionCall!(Use World Position Offsets) {materialFunction=/Engine/Functions/Engine_MaterialFunctions02/Utility/CameraVectorWithWPOOptions}
  n4 = DotProduct(A=@n3.Result, B=VectorParameter() {parameterName=Axis, defaultValue=<rgba 0.0625 0.125 0 0>}.RGB)
  n5 = Power(Base=OneMinus(None=Saturate(None=@n4)), Exp=ScalarParameter() {defaultValue=35, parameterName=Axis Power 1})
  n6 = DotProduct(A=@n3.Result, B=Multiply(A=VectorParameter() {parameterName=Axis, defaultValue=<rgba 0.0625 0.125 0 0>}.RGB) {constB=-1})
  n7 = Power(Base=OneMinus(None=Saturate(None=@n6)), Exp=ScalarParameter() {defaultValue=35, parameterName=Axis Power 2})
  n8 = Multiply(
   A=TextureCoordinate()
   B=ComponentMask(None=VectorParameter() {parameterName=End Fade, defaultValue=<rgba 1.5062 0.7813 0 0>}.RGB) {g=true, r=true}
  )
  n9 = Add(A=Time(), B=NamedRerouteUsage())
  n10 = Multiply(A=@n9, B=ScalarParameter() {defaultValue=-0.025, parameterName=Time Speed 2})
  n11 = Multiply(
   A=TextureCoordinate()
   B=ComponentMask(None=VectorParameter() {parameterName=End Fade Broad, defaultValue=<rgba 0.8 0.5 0 0>}.RGB) {g=true, r=true}
  )
  n12 = Multiply(A=@n9, B=ScalarParameter() {defaultValue=-0.0125, parameterName=Time Speed})
  n13 = Add(
   A=TextureSample(UVs=Panner(Coordinate=@n8, Time=@n10) {speedX=1}) {texture=/Game/Effects/Fort_Effects/Textures/Smoke/T_Clouds_01}.G
   B=TextureSample(UVs=Panner(Coordinate=@n11, Time=@n12) {speedX=1}) {texture=/Game/Effects/Fort_Effects/Textures/Smoke/T_Clouds_02_M}.G
  )
  n14 = Constant()
  n15 = MaterialFunctionCall!("Xb1,PS4"=@n13, "PC-High&&Epic,XSX,PS5"=@n13, "PC-Perf,Android"=@n14, Switch=@n14, "PC-Low(NotPerf)&&Medium"=@n14) {materialFunction=/Game/Effects/Niagara/MasterMaterials/Functions/Common/PlatformSwitch_3D}
  n16 = TextureCoordinate()
  n17 = Power(Base=OneMinus(None=ComponentMask(None=@n16) {r=true}), Exp=ScalarParameter() {defaultValue=1, parameterName=Ribbon Fade})
  n18 = TextureCoordinate()
  n19 = Multiply(A=ComponentMask(None=VectorParameter() {parameterName=Param_1, defaultValue=<rgba 4 1 0 0>}.RGB) {g=true, r=true}, B=TextureCoordinate())
  n20 = Multiply(A=Time(), B=ScalarParameter() {defaultValue=0.05, parameterName=detail noise pan speed})
  n21 = Multiply(
   A=ScalarParameter() {defaultValue=0.01, parameterName=Param_7}
   B=TextureSample(UVs=Panner(Coordinate=@n19, Time=@n20) {speedX=-1}) {texture=&AtmosphericCloudNoise01}.G
  )
  n22 = Constant()
  n23 = QualitySwitch(Default=@n21, Low=@n22, High=@n21, Medium=@n22, Epic=@n21)
  n24 = Add(A=Time(), B=NamedRerouteUsage())
  n25 = ScalarParameter() {defaultValue=-0.0031, parameterName=Time Scale}
  n26 = Multiply(
   A=Panner(Coordinate=Add(A=@n18, B=@n23), Time=Multiply(A=@n24, B=@n25)) {speedX=1}
   B=ComponentMask(None=VectorParameter() {parameterName=Tiling, defaultValue=<rgba 0.45 1 0 0>}.RGB) {g=true, r=true}
  )
  n27 = TextureCoordinate()
  n28 = Multiply(A=ComponentMask(None=VectorParameter() {parameterName=Param_1, defaultValue=<rgba 4 1 0 0>}.RGB) {g=true, r=true}, B=TextureCoordinate())
  n29 = Multiply(A=Time(), B=ScalarParameter() {defaultValue=0.05, parameterName=detail noise pan speed 2})
  n30 = Multiply(
   A=ScalarParameter() {defaultValue=0.01, parameterName=Param_72}
   B=TextureSample(UVs=Panner(Coordinate=@n28, Time=@n29) {speedX=-1}) {texture=&AtmosphericCloudNoise01}.G
  )
  n31 = Constant()
  n32 = QualitySwitch(Default=@n30, Low=@n31, High=@n30, Medium=@n31, Epic=@n30)
  n33 = Add(A=Time(), B=NamedRerouteUsage())
  n34 = ScalarParameter() {defaultValue=-0.0031, parameterName=Time Scale 2}
  n35 = Multiply(
   A=Panner(Coordinate=Add(A=@n27, B=@n32), Time=Multiply(A=@n33, B=@n34)) {speedX=1}
   B=ComponentMask(None=VectorParameter() {parameterName=Tiling 2, defaultValue=<rgba 2 1 0 0>}.RGB) {g=true, r=true}
  )
  n36 = TextureObjectParameter(UVs=@n35) {parameterName=TextureObject2, samplerType=SAMPLERTYPE_Grayscale, texture=&T_Smoke_Trail_Soft}
  n37 = LinearInterpolate(
   A=TextureSample(
    UVs=@n26
    Tex=TextureObjectParameter(UVs=@n26) {parameterName=TextureObject, texture=/Game/Effects/Fort_Effects/Textures/Smoke/T_Puffy_Smoke_Trail_01}
   ).G
   B=TextureSample(UVs=@n35, Tex=@n36).G
  )
  n38 = ScalarParameter() {defaultValue=1.5}
  n39 = Multiply(
   A=Power(Base=LinearInterpolate(A=@n15.Result, Alpha=@n17), Exp=ScalarParameter() {defaultValue=3, parameterName=End Power})
   B=Multiply(A=@n17, B=Multiply(A=@n37, B=@n38))
  )
  n40 = DepthFade(Opacity=@n39, FadeDistance=ScalarParameter() {defaultValue=3500, parameterName=Blend Depth})
  n41 = MaterialFunctionCall(Distance=ScalarParameter() {defaultValue=15000, parameterName=Near Camera Fade}) {materialFunction=&Near_Camera_Fade}
  n42 = Multiply(
   A=Multiply(
    A=MaterialFunctionCall("Xb1,PS4"=@n40, "PC-High&&Epic,XSX,PS5"=@n40, "PC-Perf,Android"=@n39, Switch=@n39, "PC-Low(NotPerf)&&Medium"=@n39) {materialFunction=&PlatformSwitch_3D}.Result
    B=ScalarParameter() {defaultValue=2.5, parameterName=Opacity}
   )
   B=@n41.Result
  )
  NamedRerouteDeclaration(None=Multiply(A=@n5, B=Multiply(A=@n7, B=@n42))) {name=Axis Power 1, nodeColor=<rgba 0 0.5781 1 1>}
  n43 = ScalarParameter() {parameterName=Hue Shift}
  n44 = VectorParameter() {parameterName=Albedo, defaultValue=<rgba 0 0.7833 1 1>}
  n45 = MaterialFunctionCall(
   Color_Morning=@n44.RGB
   Color_Day=@n44.RGB
   Color_Evening=VectorParameter() {parameterName=Albedo Evening, defaultValue=<rgba 0 0.7833 1 1>}.RGB
   Color_Night=VectorParameter() {parameterName=Albedo Night, defaultValue=<rgba 0 0.825 1.65 1>}.RGB
  ) {materialFunction=&Color_Time_of_Day}
  n46 = TextureCoordinate()
  n47 = Power(Base=OneMinus(None=ComponentMask(None=@n46) {r=true}), Exp=ScalarParameter() {defaultValue=1, parameterName=Param_4})
  n48 = Saturate(None=DotProduct(A=NamedRerouteUsage(), B=AtmosphericLightVector()))
  n49 = Constant() {r=1}
  n50 = QualitySwitch(Default=@n48, Low=@n49, High=@n49, Medium=@n49, Epic=@n48)
  n51 = LinearInterpolate(
   A=Multiply(A=ScalarParameter() {defaultValue=0.4, parameterName=Darken Albedo}, B=@n45.Result) {constA=0.5}
   B=@n45.Result
   Alpha=Add(A=Multiply(A=@n47, B=@n50), B=NamedRerouteUsage())
  )
  n52 = ComponentMask(None=TextureCoordinate()) {r=true}
  n53 = LinearInterpolate(
   A=Desaturation(
    None=MaterialFunctionCall(Hue Shift Percentage=@n43, Texture=@n51) {materialFunction=&HueShift}.Result
    Fraction=ScalarParameter() {defaultValue=0.5, parameterName=Param_6}
   )
   B=@n51
   Alpha=Power(Base=OneMinus(None=@n52), Exp=ScalarParameter() {defaultValue=0.75, parameterName=Power})
  )
  n54 = TextureCoordinate()
  n55 = Multiply(
   A=ScalarParameter() {defaultValue=0.5, parameterName=Param_10}
   B=Power(Base=ComponentMask(None=@n54) {r=true}, Exp=ScalarParameter() {defaultValue=3, parameterName=Param_9})
  )
  NamedRerouteDeclaration(None=Multiply(A=Add(A=@n53, B=@n55), B=NamedRerouteUsage())) {name=Emissive, nodeColor=<rgba 0.8984 1 0 1>}
  NamedRerouteDeclaration(None=@n7) {name=Axis Fade, nodeColor=<rgba 0.2891 1 0 1>}
  NamedRerouteDeclaration(None=@n5) {name=Axis Fade_1, nodeColor=<rgba 1 0 0.6094 1>}
  n56 = ScalarParameter() {defaultValue=16342.8936, parameterName=Tip}
  n57 = Power(Base=ComponentMask(None=TextureCoordinate()) {r=true}, Exp=ScalarParameter() {defaultValue=0.1, parameterName=Squeeze Power})
  n58 = Power(Base=NamedRerouteUsage()) {constExponent=0.5}
  n59 = MaterialFunctionCall(
   WidthBase=Multiply(A=LinearInterpolate(B=@n56, Alpha=@n57), B=@n58)
   WidthTip=Multiply(A=ScalarParameter() {defaultValue=1800, parameterName=Base}, B=@n58)
   Expand U or V UV Channel=StaticBoolParameter() {parameterName="Expand U or V UV Channel (SB)"}
  ) {materialFunction=&SplineThicken}
  NamedRerouteDeclaration(None=@n59.Normal) {name=New WPO Normal, nodeColor=<rgba 0.3594 0 1 1>}
  NamedRerouteDeclaration(None=@n41.Result) {name=Camera Dist, nodeColor=<rgba 1 0 0.4219 1>}
  n60 = DotProduct(A=WorldPosition().XYZ, B=Constant3Vector() {constant=<rgba 12 78 45 0>})
  n61 = ScalarParameter() {defaultValue=43500, parameterName=Param_8}
  n62 = Frac(None=Sine(None=Multiply(A=@n60, B=@n61)))
  NamedRerouteDeclaration(None=QualitySwitch(Default=@n62, Low=Constant(), Epic=@n62)) {name=Random Offset based on WP, nodeColor=<rgba 0 0.2031 1 1>}
  n63 = LinearInterpolate(
   B=Transform(None=VectorParameter() {defaultValue=<rgba 3500 0 0 0>}.RGB) {transformSourceType=TRANSFORMSOURCE_Local}
   Alpha=ComponentMask(None=TextureCoordinate()) {r=true}
  )
  n64 = ScalarParameter() {defaultValue=250, parameterName=Camera Offset}
  NamedRerouteDeclaration(None=Add(
    A=Add(A=@n59.WorldPositionOffset, B=@n63)
    B=MaterialFunctionCall(Offset Amount=@n64) {materialFunction=/Engine/Functions/Engine_MaterialFunctions02/WorldPositionOffset/CameraOffset}.Camera Offset
   )) {name=WPO, nodeColor=<rgba 1 0 0.2344 1>}
  NamedRerouteDeclaration(None=@n59.WorldPositionOffset) {name=Spline, nodeColor=<rgba 1 0.2109 0 1>}
  n65 = NamedRerouteUsage()
  n66 = VertexInterpolator(VS=PreSkinnedNormal())
  n67 = VertexInterpolator(VS=NamedRerouteUsage())
  n68 = Subtract(
   A=VertexInterpolator(VS=LocalPosition() {localOrigin=InstancePreSkinning}.XYZ).PS
   B=Add(A=WorldPosition() {worldPositionShaderOffset=WPT_CameraRelative}.XYZ, B=NamedRerouteUsage())
  )
  n69 = Multiply(A=NamedRerouteUsage(), B=ScalarParameter() {defaultValue=3790.0808, parameterName=Param_2}) {constB=128}
  n70 = Subtract(
   A=Subtract(A=VertexInterpolator(VS=@n65).PS, B=Add(A=@n66.PS, B=@n67.PS))
   B=Divide(A=Subtract(A=@n68, B=@n69), B=ScalarParameter() {defaultValue=2048, parameterName=Param_3}) {constB=1000}
  )
  NamedRerouteDeclaration(None=Normalize(VectorInput=Transform(None=@n70) {transformSourceType=TRANSFORMSOURCE_Local})) {name=normal Recalc, nodeColor=<rgba 0.5234 1 0 1>}
  n71 = CameraVectorWS()
  n72 = Saturate(None=DotProduct(A=Multiply(A=@n71) {constB=-1}, B=SkyAtmosphereLightDirection()))
  NamedRerouteDeclaration(None=Multiply(A=@n72, B=@n72)) {name=BackFace, nodeColor=<rgba 1 0.4922 0 1>}
  NamedRerouteDeclaration(None=MaterialFunctionCall(
    Color_Morning=VectorParameter() {parameterName="Color_Morning (V3)", defaultValue=<rgba 2.5 1.1397 1.4897 1>}.RGB
    Color_Day=VectorParameter() {parameterName="Color_Day (V3)", defaultValue=<rgba 1.2 1.2 1.2 0>}.RGB
    Color_Evening=VectorParameter() {parameterName="Color_Evening (V3)", defaultValue=<rgba 1.25 0.9506 0.8687 1>}.RGB
    Color_Night=VectorParameter() {parameterName="Color_Night (V3)", defaultValue=<rgba 0.6545 0.6545 0.6545 1>}.RGB
   ) {materialFunction=&Color_Time_of_Day}.Result) {name=TOD, nodeColor=<rgba 1 0.7969 0 1>}
  NamedRerouteUsage()
  NamedRerouteUsage()
  NamedRerouteUsage()
  NamedRerouteUsage()
  NamedRerouteUsage()
  Subtract(A=CameraPositionWS(), B=WorldPosition() {worldPositionShaderOffset=WPT_CameraRelative}.XYZ) {constB=-1}
  VectorParameter() {parameterName=Albedo, defaultValue=<rgba 0 0.7833 1 1>}
  VertexInterpolator()
  VertexInterpolator()
  WorldPosition()
 ]
}
