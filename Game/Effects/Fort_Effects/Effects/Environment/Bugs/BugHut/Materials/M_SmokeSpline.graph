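# Translucent, unlit, two-sided material for a smoke plume in the BugHut environment effects, built on a spline-thickened strip.
# World position offset math drifts the strip along a vector by a powered V gradient, widens it with SplineThicken and adds a camera offset.
# A panned, axis-swizzled smoke texture, time-offset by vertex green, drives opacity with V-end, depth and near-camera fades.
# Color is tinted by time of day; alternative bend and tangent-based thickening branches also exist.
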Material /Game/Effects/Fort_Effects/Effects/Environment/Bugs/BugHut/Materials/M_SmokeSpline
graph {
 properties {bAutomaticallySetUsageInEditor=false, blendMode=BLEND_Translucent, shadingModel=MSM_Unlit, twoSided=true, usedShadingModels=Unlit}
 nodes [
  n1 = ScalarParameter() {defaultValue=266.6728, parameterName=Param_6}
  n2 = ScalarParameter() {defaultValue=958.5067, parameterName=Param_7}
  n3 = ComponentMask(None=TextureCoordinate()) {g=true}
  n4 = LinearInterpolate(
   B=VectorParameter() {defaultValue=<rgba -250 0 758.5912 0>}.RGB
   Alpha=Power(Base=OneMinus(None=@n3), Exp=ScalarParameter() {defaultValue=3, parameterName=Param_3})
  )
  n5 = Add(A=WorldPosition().XYZ, B=@n4)
  Add(
   A=Add(
    A=MaterialFunctionCall(WidthBase=@n1, WidthTip=@n2, WorldPosition=@n5) {materialFunction=&SplineThicken}.WorldPositionOffset
    B=@n4
   )
   B=MaterialFunctionCall(Offset Amount=ScalarParameter() {defaultValue=150, parameterName=Param_1}) {materialFunction=/Engine/Functions/Engine_MaterialFunctions02/WorldPositionOffset/CameraOffset}.Camera Offset
  )
  Add(B=MaterialFunctionCall(Offset Amount=ScalarParameter() {parameterName=CameraOffset}, World Position=WorldPosition().XYZ) {materialFunction=&CameraOffset}.Camera Offset)
  n6 = Power(Base=ComponentMask(None=TextureCoordinate()) {g=true}, Exp=ScalarParameter() {defaultValue=1, parameterName=Bend Power})
  Add(A=LinearInterpolate(A=VectorParameter() {parameterName=root}.RGB, B=VectorParameter() {parameterName=Bend}.RGB, Alpha=OneMinus(None=@n6)))
  Multiply(
   A=VectorParameter() {parameterName=Param_1, defaultValue=<rgba 0.0808 0.5097 1 1>}.RGB
   B=MaterialFunctionCall(
    Color_Morning=VectorParameter() {parameterName="Color_Morning (V3)", defaultValue=<rgba 1 0.758 0.6473 1>}.RGB
    Color_Day=VectorParameter() {parameterName="Color_Day (V3)", defaultValue=<rgba 1 0.6753 0.6753 0>}.RGB
    Color_Evening=VectorParameter() {parameterName="Color_Evening (V3)", defaultValue=<rgba 1 0.7546 0.6713 1>}.RGB
    Color_Night=VectorParameter() {parameterName="Color_Night (V3)", defaultValue=<rgba 0.5283 0.5485 1 1>}.RGB
   ) {materialFunction=&Color_Time_of_Day}.Result
  )
  Multiply(B=ScalarParameter() {defaultValue=1, parameterName=Param_8})
  n7 = Multiply(A=TextureCoordinate(), B=ComponentMask(None=VectorParameter() {parameterName=C, defaultValue=<rgba 1 1 0 0>}.RGB) {g=true, r=true})
  n8 = Multiply(A=Add(A=Time(), B=VertexColor().G), B=ScalarParameter() {defaultValue=0.025})
  n9 = TextureSample(UVs=MaterialFunctionCall(XY=Panner(Coordinate=@n7, Time=@n8) {speedY=1}) {materialFunction=&Swizzle}.YX) {texture=/Game/Effects/Fort_Effects/Textures/Smoke/T_Puffy_Smoke_Trail_01}
  n10 = ScalarParameter() {defaultValue=2, parameterName=Param_2}
  n11 = TextureCoordinate()
  n12 = ComponentMask(None=TextureCoordinate()) {g=true}
  n13 = Multiply(
   A=Multiply(A=Multiply(A=@n9.G, B=@n10), B=ComponentMask(None=@n11) {g=true})
   B=Power(Base=OneMinus(None=@n12), Exp=ScalarParameter() {defaultValue=0.5, parameterName=Param_4})
  )
  Multiply(
   A=Multiply(A=DepthFade(Opacity=@n13), B=ScalarParameter() {defaultValue=2, parameterName=Param_5})
   B=MaterialFunctionCall(Distance=ScalarParameter() {defaultValue=1750, parameterName=Param_9}) {materialFunction=&Near_Camera_Fade}.Result
  )
  n14 = Subtract(A=ComponentMask(None=TextureCoordinate()) {r=true}) {constB=0.5}
  n15 = MaterialFunctionCall() {materialFunction=&TangentBasis}
  n16 = ScalarParameter() {defaultValue=353.3233, parameterName=Base}
  n17 = ScalarParameter() {defaultValue=1243.3206, parameterName=End}
  n18 = OneMinus(None=ComponentMask(None=TextureCoordinate()) {g=true})
  NamedRerouteDeclaration(None=Multiply(A=Multiply(A=@n14, B=@n15.X U), B=LinearInterpolate(A=@n16, B=@n17, Alpha=@n18))) {name=Tangent Based Spline Thicken, nodeColor=<rgba 0.5 1 0 1>}
  NamedRerouteUsage()
  TransformPosition() {transformType=TRANSFORMPOSSOURCE_World}
 ]
}
