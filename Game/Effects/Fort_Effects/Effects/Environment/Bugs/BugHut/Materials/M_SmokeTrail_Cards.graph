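# Translucent unlit material for animated smoke trail cards in the BugHut environment effects.
# Looping trail masks feed waving functions and vertex math that orients cards from camera and wave tangent, growing them along the trail.
# A rotated SubUV flipbook drives opacity, softened by depth fade, sphere mask and near-camera fade.
# Custom HLSL ray-marches the flipbook in 2D along the atmospheric light direction for backlit shading with time-of-day color and density.
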Material /Game/Effects/Fort_Effects/Effects/Environment/Bugs/BugHut/Materials/M_SmokeTrail_Cards
graph {
 properties {bAutomaticallySetUsageInEditor=false, blendMode=BLEND_Translucent, shadingModel=MSM_Unlit, usedShadingModels=Unlit}
 nodes [
  Constant3Vector() {constant=<rgba 0 0 1 0>}
  n1 = AtmosphericLightVector()
  CrossProduct(A=@n1, B=NamedRerouteUsage())
  n2 = MaterialFunctionCall(
   Color_Morning=VectorParameter() {parameterName=Albedo Morning, defaultValue=<rgba 0 0.6936 1 1>}.RGB
   Color_Day=VectorParameter() {parameterName=Albedo, defaultValue=<rgba 2.058 6.5509 7 1>}.RGB
   Color_Evening=VectorParameter() {parameterName=Albedo Evening, defaultValue=<rgba 0.1059 0.553 1 1>}.RGB
   Color_Night=VectorParameter() {parameterName=Albedo Night, defaultValue=<rgba 0 0.5 1 1>}.RGB
  ) {materialFunction=&Color_Time_of_Day}
  n3 = AtmosphericLightVector()
  n4 = NamedRerouteUsage()
  LinearInterpolate(
   A=Multiply(A=@n2.Result, B=ScalarParameter() {defaultValue=0.4, parameterName=Albdeo Darken})
   B=@n2.Result
   Alpha=Saturate(None=DotProduct(A=@n3, B=@n4))
  )
  n5 = Add(A=NamedRerouteUsage(), B=Multiply(A=LocalPosition().XYZ) {constB=-1})
  n6 = Transform(None=WorldPosition() {worldPositionShaderOffset=WPT_CameraRelative}.XYZ) {transformSourceType=TRANSFORMSOURCE_World, transformType=TRANSFORM_Local}
  NamedRerouteDeclaration(None=Normalize(VectorInput=Add(A=@n5, B=@n6))) {name=Card Vector, nodeColor=<rgba 1 0.6797 0 1>}
  n7 = NamedRerouteUsage()
  n8 = Time()
  n9 = ScalarParameter() {defaultValue=1, parameterName=Wiggle Time Scale}
  n10 = VectorParameter() {defaultValue=<rgba -0.25 0.05 0 1>}
  n11 = Add(
   A=AppendVector(A=Multiply(A=@n7) {constB=0.6}, B=Constant())
   B=Multiply(A=Multiply(A=@n8, B=@n9), B=ComponentMask(None=@n10.RGB) {g=true, r=true}) {constB=-0.15}
  )
  n12 = MaterialFunctionCall!(
   UVs=@n11
   Offset=Constant() {r=0.02}
   Linear Mask=NamedRerouteUsage()
   Non Linear Mask=NamedRerouteUsage()
   Tangent Multiplier=Constant() {r=0.5}
  ) {materialFunction=/Game/Environments/TechArt/SmokeTrail/MF_SmokeTrail_Waving}
  n13 = MaterialFunctionCall(
   UVs=@n11
   Offset=Constant() {r=-0.02}
   Linear Mask=NamedRerouteUsage()
   Non Linear Mask=NamedRerouteUsage()
   Tangent Multiplier=Constant() {r=0.5}
  ) {materialFunction=&MF_SmokeTrail_Waving}
  NamedRerouteDeclaration(None=Normalize(VectorInput=Subtract(A=@n12.Result, B=@n13.Result))) {name=Waving Tangent, nodeColor=<rgba 0 1 0.8594 1>}
  n14 = ScalarParameter() {defaultValue=4, parameterName=SubUV}
  n15 = Add(
   A=ComponentMask(None=TextureCoordinate() {coordinateIndex=1}) {r=true}
   B=Multiply(A=Time(), B=ScalarParameter() {defaultValue=0.01, parameterName=Delta Time Scale})
  )
  n16 = ScalarParameter() {defaultValue=1, parameterName=Loop Duration}
  n17 = LinearInterpolate(Alpha=ComponentMask(None=TextureCoordinate() {coordinateIndex=1}) {g=true}) {constA=-360, constB=360}
  MaterialExpressionMaterialFunctionCall_5 = MaterialFunctionCall!(
   U_Images=@n14
   V_Images=@n14
   SubUV Texture RGB
   Time Speed=ScalarParameter() {defaultValue=1, parameterName=SubUV Time Scale}
   Manual Time Control=StaticBoolParameter() {defaultValue=true}
   Time=Divide(A=Fmod(A=@n15, B=@n16), B=@n16)
   UVs=Rotator(Coordinate=TextureCoordinate(), Time=@n17)
  ) {materialFunction=/Game/Effects/Fort_Effects/Materials/_Functions/MF_SubUV}
  MaterialExpressionTextureObjectParameter_0 = TextureObjectParameter(UVs=@MaterialExpressionMaterialFunctionCall_5.UVs) {parameterName=TextureObject, texture=/Game/Effects/Niagara/MasterMaterials/Smoke/Textures/NEW/RT4x4_4_Tex}
  n18 = NamedRerouteUsage()
  n19 = Multiply(
   A=DepthFade(Opacity=TextureSample(UVs=@MaterialExpressionMaterialFunctionCall_5.UVs, Tex=@MaterialExpressionTextureObjectParameter_0).B)
   B=SphereMask(
    A=OneMinus(None=@n18)
    B=ScalarParameter() {parameterName=Start}
    Radius=ScalarParameter() {defaultValue=2.0676, parameterName=Radius}
    Hardness=ScalarParameter() {parameterName=Hardness}
   )
  )
  n20 = ScalarParameter() {defaultValue=1, parameterName=Overall Opacity}
  n21 = ScalarParameter() {parameterName=Near Camera Fade}
  NamedRerouteDeclaration(None=Multiply(
    A=Multiply(A=@n19, B=@n20)
    B=MaterialFunctionCall(Distance=@n21) {materialFunction=&Near_Camera_Fade}.Result
   )) {name=Opacity, nodeColor=<rgba 1 0.8672 0 1>}
  NamedRerouteDeclaration() {name=Albedo, nodeColor=<rgba 1 0 0.8203 1>}
  n22 = NamedRerouteUsage()
  n23 = VectorParameter() {parameterName=Noise Pan Dir}
  n24 = Add(A=WorldPosition().XY, B=Multiply(A=Time(), B=ComponentMask(None=@n23.RGB) {g=true, r=true}))
  n25 = ScalarParameter() {defaultValue=0.0025, parameterName=Noise Scale}
  n26 = Power(
   Base=TextureSample(UVs=Multiply(A=@n24, B=@n25)) {texture=&AtmosphericCloudNoise01}.G
   Exp=ScalarParameter() {defaultValue=0.5, parameterName=Param_2}
  )
  NamedRerouteDeclaration(None=Multiply(A=LinearInterpolate(A=@n22, Alpha=@n26), B=ScalarParameter() {defaultValue=3})) {name=Erode, nodeColor=<rgba 1 0 0.3984 1>}
  n27 = TextureCoordinate() {coordinateIndex=1}
  n28 = Time()
  n29 = ScalarParameter() {defaultValue=0.05, parameterName=time scale}
  n30 = Frac(None=Add(A=ComponentMask(None=@n27) {r=true}, B=Multiply(A=@n28, B=@n29) {constB=0.1}))
  NamedRerouteDeclaration(None=Power(
    Base=NamedRerouteDeclaration(None=@n30) {name=Linear Mask, nodeColor=<rgba 0.9453 0 1 1>}
    Exp=ScalarParameter() {defaultValue=1.6, parameterName=Shape Exponent}
   ) {constExponent=1.6}) {name=Non Linear Mask, nodeColor=<rgba 0.1016 1 0 1>}
  NamedRerouteDeclaration(None=LinearInterpolate(
    A=ScalarParameter() {defaultValue=1, parameterName=Width Start}
    B=ScalarParameter() {defaultValue=2, parameterName=Width End}
    Alpha=NamedRerouteUsage()
   )) {name=Card Width, nodeColor=<rgba 1 0 0.4219 1>}
  NamedRerouteDeclaration(None=LinearInterpolate(
    A=ScalarParameter() {defaultValue=0.5, parameterName=Size Start}
    B=ScalarParameter() {defaultValue=5, parameterName=Size End}
    Alpha=NamedRerouteUsage()
   )) {name=Card Size, nodeColor=<rgba 0 0.7422 1 1>}
  n31 = NamedRerouteUsage()
  n32 = CrossProduct(A=@n31, B=NamedRerouteUsage())
  n33 = Normalize(VectorInput=@n32)
  n34 = NamedRerouteUsage()
  n35 = Multiply(A=@n33, B=@n34) {constB=0.25}
  n36 = LocalPosition()
  n37 = ComponentMask(None=@n36.XYZ) {r=true}
  n38 = Multiply(A=Multiply(A=NamedRerouteUsage(), B=VertexColor().R)) {constB=-100}
  n39 = NamedRerouteUsage()
  n40 = CrossProduct(A=@n32, B=@n31)
  n41 = Multiply(A=OneMinus(None=@n34), B=@n40)
  n42 = Multiply(A=CrossProduct(A=@n33, B=@n31), B=@n34) {constB=0.25}
  n43 = ComponentMask(None=@n36.XYZ) {g=true}
  n44 = Add(A=Add(A=Multiply(A=@n35, B=@n37), B=Multiply(A=@n38, B=@n39)), B=Multiply(A=Add(A=@n41, B=@n42), B=@n43))
  n45 = Multiply(A=@n44, B=NamedRerouteUsage())
  NamedRerouteDeclaration(None=Transform(None=Add(A=@n5, B=@n45)) {transformSourceType=TRANSFORMSOURCE_Local}) {name=WPO, nodeColor=<rgba 0.4297 0 1 1>}
  NamedRerouteDeclaration(None=MaterialFunctionCall(UVs=@n11, Offset=Constant(), Linear Mask=NamedRerouteUsage(), Non Linear Mask=NamedRerouteUsage(), Tangent Multiplier=Constant() {r=1}) {materialFunction=&MF_SmokeTrail_Waving}.Result) {name=Waving WPO, nodeColor=<rgba 1 0 0.7969 1>}
  n46 = Multiply(A=Subtract(A=@n44, B=Multiply(A=@n40, B=@n43))) {constB=0.01}
  n47 = VertexInterpolator(VS=Divide(
    A=Transform(None=@n46) {transformSourceType=TRANSFORMSOURCE_Local}
    B=MaterialFunctionCall() {materialFunction=&ObjectScale}.Scale XYZ
   ))
  NamedRerouteDeclaration(None=@n47.PS) {name=Normals, nodeColor=<rgba 0.4062 0 1 1>}
  MaterialExpressionScalarParameter_4 = ScalarParameter() {defaultValue=0.05, parameterName=March Distance}
  MaterialExpressionScalarParameter_20 = ScalarParameter() {defaultValue=32, parameterName=NumSteps}
  MaterialExpressionScalarParameter_5 = ScalarParameter() {defaultValue=1, parameterName=TL}
  MaterialExpressionMultiply_35 = Multiply(A=ScalarParameter() {defaultValue=0.25, parameterName=Shaodw Darken}, B=NamedRerouteUsage())
  MaterialExpressionMultiply_11 = Multiply(
   A=ScalarParameter() {defaultValue=1, parameterName=Overall Density}
   B=MaterialFunctionCall(
    Value_Morning=ScalarParameter() {defaultValue=1, parameterName=Density Scale Morning}
    Value_Day=ScalarParameter() {defaultValue=1, parameterName=Density Scale Day}
    Value_Evening=ScalarParameter() {defaultValue=1, parameterName=Density Scale Eveneing}
    Value_Night=ScalarParameter() {defaultValue=1, parameterName=Density Scale Night}
   ) {materialFunction=&Scalar_Time_of_Day}.Result
  )
  MaterialExpressionRotator_1 = Rotator(Coordinate=Transform(None=@n1) {transformSourceType=TRANSFORMSOURCE_World, transformType=TRANSFORM_Tangent}, Time=@n17)
  MaterialExpressionMaterialFunctionCall_4 = MaterialFunctionCall(
   Color_Morning=VectorParameter() {parameterName="Color_Morning (V3)", defaultValue=<rgba 0.1788 0.066 0.0625 1>}.RGB
   Color_Day=VectorParameter() {parameterName="Color_Day (V3)", defaultValue=<rgba 0.1198 0.0952 0.0834 1>}.RGB
   Color_Evening=VectorParameter() {parameterName="Color_Evening (V3)", defaultValue=<rgba 0.1892 0.0252 0.0306 1>}.RGB
   Color_Night=VectorParameter() {parameterName="Color_Night (V3)", defaultValue=<rgba 0 0.0762 0.1545 1>}.RGB
  ) {materialFunction=&Color_Time_of_Day}
  MaterialExpressionScalarParameter_30 = ScalarParameter() {defaultValue=8, parameterName=Contrast}
  n48 = Custom!(
   Tex=@MaterialExpressionTextureObjectParameter_0
   marchDistance=@MaterialExpressionScalarParameter_4
   numSteps=@MaterialExpressionScalarParameter_20
   TL=@MaterialExpressionScalarParameter_5
   LO=@MaterialExpressionMultiply_35
   density=@MaterialExpressionMultiply_11
   UV=@MaterialExpressionMaterialFunctionCall_5.UVs
   marchDir=@MaterialExpressionRotator_1
   LightColor=@MaterialExpressionMaterialFunctionCall_4.Result
   contrast=@MaterialExpressionScalarParameter_30
  ) {
   code="for (float d = 0.0; d < marchDistance; d += marchDistance / numSteps)\r\n{\r\n    UV += (marchDir * d) / numSteps;\r\n    float4 textureSample = Texture2DSampleLevel(Tex, TexSampler, UV.xy, 0) / (numSteps - contrast);\r\n    // Extracting the red channel\r\n    float densityChannel = textureSample.b;\r\n    TL *= saturate(1.0 - densityChannel);\r\n}\r\nfloat3 LI = LightColor * TL.xxx;\r\nLO += LI * density;\r\nreturn LO;"
   description=2D RayMarching
   inputs [
    {
     inputName=Tex
     input {expression=&self:MaterialExpressionTextureObjectParameter_0}
    }
    {
     inputName=marchDistance
     input {expression=&self:MaterialExpressionScalarParameter_4}
    }
    {
     inputName=numSteps
     input {expression=&self:MaterialExpressionScalarParameter_20}
    }
    {
     inputName=TL
     input {expression=&self:MaterialExpressionScalarParameter_5}
    }
    {
     inputName=LO
     input {expression=&self:MaterialExpressionMultiply_35}
    }
    {
     inputName=density
     input {expression=&self:MaterialExpressionMultiply_11}
    }
    {
     inputName=UV
     input {
      expression=&self:MaterialExpressionMaterialFunctionCall_5
      outputIndex=2
     }
    }
    {
     inputName=marchDir
     input {expression=&self:MaterialExpressionRotator_1}
    }
    {
     inputName=LightColor
     input {expression=&self:MaterialExpressionMaterialFunctionCall_4}
    }
    {
     inputName=contrast
     input {expression=&self:MaterialExpressionScalarParameter_30}
    }
   ]
  }
  n49 = ScalarParameter() {defaultValue=1, parameterName=Backlight Power}
  n50 = Multiply(
   A=Multiply(A=Power(Base=@n48, Exp=@n49), B=ScalarParameter() {defaultValue=1, parameterName=Backlight Scale})
   B=@MaterialExpressionMaterialFunctionCall_4.Result
  )
  NamedRerouteDeclaration(None=Multiply(B=Add(A=@n48, B=@n50))) {name=Emissive, nodeColor=<rgba 0 1 0.3906 1>}
  NamedRerouteUsage()
  NamedRerouteUsage()
  NamedRerouteUsage()
  NamedRerouteUsage()
  OneMinus(None=NamedRerouteUsage())
  Reroute(None=@n47.PS)
  TextureCoordinate()
  VectorParameter() {parameterName=Albedo, defaultValue=<rgba 2.058 6.5509 7 1>}
  VertexInterpolator(VS=NamedRerouteUsage())
 ]
}
