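# Translucent, unlit, two-sided smoke material for spline-based plumes in the BugHut bug environment effects.
# SplineThicken widens the geometry from base to tip, and its world position offset is summed with a CameraOffset.
# Color_Time_of_Day blends albedo parameters for morning/day, evening and night.
# A mask multiplies a soft U-edge falloff, a V gradient and a panning, power-adjusted cloud texture, then applies view-angle and near-camera fades.
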Material /Game/Effects/Fort_Effects/Effects/Environment/Bugs/BugHut/Materials/M_SplineSmoke_Bug
graph {
 properties {bAutomaticallySetUsageInEditor=false, blendMode=BLEND_Translucent, shadingModel=MSM_Unlit, twoSided=true, usedShadingModels=Unlit}
 nodes [
  Add(
   A=MaterialFunctionCall(WidthBase=ScalarParameter() {defaultValue=1439.567}, WidthTip=ScalarParameter() {defaultValue=12144.6025, parameterName=Param_1}) {materialFunction=&SplineThicken}.WorldPositionOffset
   B=MaterialFunctionCall(Offset Amount=ScalarParameter() {parameterName=Param_2}, World Position=WorldPosition().XYZ) {materialFunction=/Engine/Functions/Engine_MaterialFunctions02/WorldPositionOffset/CameraOffset}.Camera Offset
  )
  Add()
  n1 = VectorParameter() {parameterName=Albedo, defaultValue=<rgba 0 0.7833 1 1>}
  MaterialFunctionCall(
   Color_Morning=@n1.RGB
   Color_Day=@n1.RGB
   Color_Evening=VectorParameter() {parameterName=Albedo Evening, defaultValue=<rgba 0 0.7833 1 1>}.RGB
   Color_Night=VectorParameter() {parameterName=Albedo Night, defaultValue=<rgba 0 0.5 1 1>}.RGB
  ) {materialFunction=&Color_Time_of_Day}
  n2 = ComponentMask(None=TextureCoordinate()) {g=true}
  n3 = ComponentMask(None=TextureCoordinate()) {r=true}
  n4 = Multiply(A=Multiply(A=OneMinus(None=@n3), B=@n3), B=ScalarParameter() {defaultValue=1, parameterName=Param_3})
  n5 = TextureCoordinate()
  n6 = ComponentMask(None=VectorParameter() {parameterName=C, defaultValue=<rgba 0.5 2 0 0>}.RGB) {g=true, r=true}
  n7 = Time()
  n8 = ScalarParameter() {defaultValue=0.01, parameterName=Param_6}
  n9 = TextureSample(UVs=Panner(Coordinate=Multiply(A=@n5, B=@n6), Time=Multiply(A=@n7, B=@n8)) {speedY=1}) {texture=/Game/Effects/Fort_Effects/Textures/Smoke/T_Clouds_01}
  n10 = ScalarParameter() {defaultValue=1, parameterName=Param_5}
  n11 = Multiply(A=Multiply(A=Saturate(None=@n4), B=Power(Base=@n9.G, Exp=@n10)), B=ScalarParameter() {defaultValue=4, parameterName=Param_4})
  n12 = ScalarParameter() {defaultValue=0.25, parameterName=Param_7}
  Multiply(
   A=Multiply(
    A=Multiply(A=@n2, B=@n11)
    B=MaterialFunctionCall("FadeSensitivity (optional)"=@n12) {materialFunction=&FadeBasedOnViewAngle}.Result
   )
   B=MaterialFunctionCall(Distance=ScalarParameter() {defaultValue=75000, parameterName=Near Cam Fade}) {materialFunction=&Near_Camera_Fade}.Result
  )
  VectorParameter() {defaultValue=<rgba 0.0747 0.7995 1 0>}
 ]
}
