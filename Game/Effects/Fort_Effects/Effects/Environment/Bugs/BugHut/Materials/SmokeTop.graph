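# Translucent, unlit, two-sided smoke ribbon material for the BugHut environment effect, presumably its top plume.
# SplineThicken, an offset blended along U and a camera offset shape the geometry.
# A panning puffy-smoke texture drives opacity with ribbon and depth fades, and feeds a custom 2D raymarch toward the atmospheric light for self-shadowing.
# Time-of-day color and density tint the lighting, which gains a backlight term and albedo to form emissive.
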Material /Game/Effects/Fort_Effects/Effects/Environment/Bugs/BugHut/Materials/SmokeTop
graph {
 properties {bAutomaticallySetUsageInEditor=false, blendMode=BLEND_Translucent, shadingModel=MSM_Unlit, twoSided=true, usedShadingModels=Unlit}
 nodes [
  n1 = MaterialFunctionCall(
   WidthBase=ScalarParameter() {defaultValue=5002.8193}
   WidthTip=ScalarParameter() {defaultValue=750, parameterName=Param_1}
   Expand U or V UV Channel=StaticBoolParameter() {parameterName="Expand U or V UV Channel (SB)"}
  ) {materialFunction=&SplineThicken}
  n2 = Transform(None=VectorParameter() {defaultValue=<rgba 8680.6611 0 0.6798 0>}.RGB) {transformSourceType=TRANSFORMSOURCE_World, transformType=TRANSFORM_Local}
  n3 = ComponentMask(None=TextureCoordinate()) {r=true}
  Add(
   A=Add(A=@n1.WorldPositionOffset, B=LinearInterpolate(B=@n2, Alpha=@n3))
   B=MaterialFunctionCall(Offset Amount=ScalarParameter() {defaultValue=50, parameterName=Param_6}) {materialFunction=/Engine/Functions/Engine_MaterialFunctions02/WorldPositionOffset/CameraOffset}.Camera Offset
  )
  n4 = DotProduct(A=CameraVectorWS(), B=NamedRerouteUsage())
  Multiply(B=Power(Base=Saturate(None=@n4), Exp=ScalarParameter() {defaultValue=1.5, parameterName=Param_4}))
  n5 = ComponentMask(None=TextureCoordinate()) {r=true}
  n6 = Time()
  n7 = ScalarParameter() {defaultValue=-0.025, parameterName=Param_3}
  MaterialExpressionMultiply_2 = Multiply(
   A=Panner(Coordinate=TextureCoordinate(), Time=Multiply(A=@n6, B=@n7)) {speedX=1}
   B=ComponentMask(None=VectorParameter() {parameterName=Param_1, defaultValue=<rgba 0.2981 1 0 0>}.RGB) {g=true, r=true}
  )
  MaterialExpressionTextureObjectParameter_0 = TextureObjectParameter(UVs=@MaterialExpressionMultiply_2) {parameterName=TextureObject, texture=/Game/Effects/Fort_Effects/Textures/Smoke/T_Puffy_Smoke_Trail_01}
  n8 = Multiply(
   A=Power(Base=OneMinus(None=@n5), Exp=ScalarParameter() {defaultValue=1.5, parameterName=Ribbon Fade})
   B=TextureSample(UVs=@MaterialExpressionMultiply_2, Tex=@MaterialExpressionTextureObjectParameter_0) {texture=&T_Puffy_Smoke_Trail_01}.G
  )
  n9 = ScalarParameter() {defaultValue=750, parameterName=Blend Depth}
  NamedRerouteDeclaration(None=Multiply(A=DepthFade(Opacity=@n8, FadeDistance=@n9), B=ScalarParameter() {defaultValue=10, parameterName=Opacity})) {name=Opacity, nodeColor=<rgba 0 0.5781 1 1>}
  n10 = VectorParameter() {parameterName=Albedo, defaultValue=<rgba 0.7066 0.2687 0.4506 1>}
  MaterialExpressionScalarParameter_2 = ScalarParameter() {defaultValue=0.1, parameterName=Param_2}
  MaterialExpressionScalarParameter_20 = ScalarParameter() {defaultValue=8, parameterName=NumSteps}
  MaterialExpressionScalarParameter_5 = ScalarParameter() {defaultValue=1.0031, parameterName=TL}
  MaterialExpressionMultiply_7 = Multiply(A=NamedRerouteUsage(), B=ScalarParameter() {defaultValue=0.25, parameterName=Param_5})
  MaterialExpressionMultiply_8 = Multiply(
   A=ScalarParameter() {defaultValue=1, parameterName=Overall Density}
   B=MaterialFunctionCall(
    Value_Morning=ScalarParameter() {defaultValue=1, parameterName=Density Scale Morning}
    Value_Day=ScalarParameter() {defaultValue=0.75, parameterName=Density Scale Day}
    Value_Evening=ScalarParameter() {defaultValue=0.5, parameterName=Density Scale Eveneing}
    Value_Night=ScalarParameter() {defaultValue=1, parameterName=Density Scale Night}
   ) {materialFunction=&Scalar_Time_of_Day}.Result
  )
  n11 = AtmosphericLightVector()
  n12 = CameraVectorWS()
  n13 = NamedRerouteUsage()
  n14 = LinearInterpolate(Alpha=Saturate(None=DotProduct(A=@n12, B=@n13))) {constA=-1}
  n15 = TextureCoordinate() {coordinateIndex=1}
  MaterialExpressionRotator_1 = Rotator(
   Coordinate=Transform(None=Multiply(A=@n11, B=@n14)) {transformSourceType=TRANSFORMSOURCE_World, transformType=TRANSFORM_Tangent}
   Time=LinearInterpolate(Alpha=ComponentMask(None=@n15) {g=true}) {constA=-360, constB=360}
  )
  MaterialExpressionMultiply_17 = Multiply(
   A=MaterialFunctionCall(
    Color_Morning=VectorParameter() {parameterName="Color_Morning (V3)", defaultValue=<rgba 0.6997 0.4615 0.4487 1>}.RGB
    Color_Day=VectorParameter() {parameterName="Color_Day (V3)", defaultValue=<rgba 2 1.7795 1.656 1>}.RGB
    Color_Evening=VectorParameter() {parameterName="Color_Evening (V3)", defaultValue=<rgba 1 0.6753 0.7109 1>}.RGB
    Color_Night=VectorParameter() {parameterName="Color_Night (V3)", defaultValue=<rgba 0.4967 0.4531 1 1>}.RGB
   ) {materialFunction=&Color_Time_of_Day}.Result
   B=MaterialFunctionCall(
    Value_Morning=ScalarParameter() {defaultValue=1, parameterName=TOD Morning}
    Value_Day=ScalarParameter() {defaultValue=0.75, parameterName=TOD Day}
    Value_Evening=ScalarParameter() {defaultValue=1, parameterName=TOD Evening}
    Value_Night=ScalarParameter() {defaultValue=1, parameterName=TOD Night}
   ) {materialFunction=&Scalar_Time_of_Day}.Result
  )
  MaterialExpressionScalarParameter_30 = ScalarParameter() {defaultValue=4, parameterName=Contrast}
  n16 = Custom!(
   Tex=@MaterialExpressionTextureObjectParameter_0
   marchDistance=@MaterialExpressionScalarParameter_2
   numSteps=@MaterialExpressionScalarParameter_20
   TL=@MaterialExpressionScalarParameter_5
   LO=@MaterialExpressionMultiply_7
   density=@MaterialExpressionMultiply_8
   UV=@MaterialExpressionMultiply_2
   marchDir=@MaterialExpressionRotator_1
   LightColor=@MaterialExpressionMultiply_17
   contrast=@MaterialExpressionScalarParameter_30
  ) {
   code="for (float d = 0.0; d < marchDistance; d += marchDistance / numSteps)\r\n{\r\n    UV += (marchDir * d) / numSteps;\r\n    float4 textureSample = Texture2DSampleLevel(Tex, TexSampler, UV.xy, 0) / (numSteps - contrast);\r\n    // Extracting the red channel\r\n    float densityChannel = textureSample.b;\r\n    TL *= saturate(1.0 - densityChannel);\r\n}\r\nfloat3 LI = LightColor * TL.xxx;\r\nLO += LI * density;\r\nreturn LO;"
   description=2D RayMarching
   inputs [
    {
     inputName=Tex
     input {expression=&self:MaterialExpressionTextureObjectParameter_0}
    }
    {
     inputName=marchDistance
     input {expression=&self:MaterialExpressionScalarParameter_2}
    }
    {
     inputName=numSteps
     input {expression=&self:MaterialExpressionScalarParameter_20}
    }
    {
     inputName=TL
     input {expression=&self:MaterialExpressionScalarParameter_5}
    }
    {inputName=LO, input {expression=&self:MaterialExpressionMultiply_7}}
    {
     inputName=density
     input {expression=&self:MaterialExpressionMultiply_8}
    }
    {inputName=UV, input {expression=&self:MaterialExpressionMultiply_2}}
    {
     inputName=marchDir
     input {expression=&self:MaterialExpressionRotator_1}
    }
    {
     inputName=LightColor
     input {expression=&self:MaterialExpressionMultiply_17}
    }
    {
     inputName=contrast
     input {expression=&self:MaterialExpressionScalarParameter_30}
    }
   ]
  }
  n17 = ScalarParameter() {defaultValue=2, parameterName=Backlight Power}
  n18 = Multiply(
   A=Multiply(A=Power(Base=@n16, Exp=@n17), B=ScalarParameter() {defaultValue=1, parameterName=Backlight Scale})
   B=@MaterialExpressionMultiply_17
  )
  NamedRerouteDeclaration(None=Multiply(A=@n10.RGB, B=Add(A=@n16, B=@n18))) {name=Emissive, nodeColor=<rgba 0.8984 1 0 1>}
  NamedRerouteDeclaration(None=@n1.Normal) {name=New WPO Normal, nodeColor=<rgba 0.3594 0 1 1>}
  NamedRerouteDeclaration(None=@n10.RGB) {name=Albedo, nodeColor=<rgba 0.9219 1 0 1>}
  NamedRerouteUsage()
  NamedRerouteUsage()
 ]
}
